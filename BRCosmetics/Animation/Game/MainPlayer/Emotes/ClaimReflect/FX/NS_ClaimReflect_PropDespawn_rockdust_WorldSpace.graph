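# Niagara rock dust puff for the prop despawn in the ClaimReflect player emote.
# Only Dust001 is enabled; Dirt, Dust and LWE_SMOKE are off, leaving User.Scale used only by disabled Dirt.
# Dust001 bursts short-lived sprites from a small sphere, moved by outward velocity, drag, gravity and slight lift, with curves animating size and dynamic material parameters.
# Time of day tints and fades the dust, and a front-end modifier drops its spawn count to zero.
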NiagaraSystem /BRCosmetics/Animation/Game/MainPlayer/Emotes/ClaimReflect/FX/NS_ClaimReflect_PropDespawn_rockdust_WorldSpace
dependencies {
 usedDataInterfaces [&NiagaraDataInterfaceCurve, &NiagaraDataInterfaceVector2DCurve]
 usedDynamicInputs []
 usedModules [&SystemState]
}
emitters {
 Dirt {
  bEnabled=false
  emitterUpdateScript [
   &EmitterState(Enable Distance Culling=true, Module.bUseMaxDistance=true, Module.bUseMinDistance=true, Scalability Mode=::NewEnumerator1)
   &SpawnBurst_Instantaneous(
    Spawn Count=&RandomRangeInt(Minimum=5, Maximum=25)
    Spawn Probability=0.5
    Use Spawn Probability=false
    UseLoopCountLimit=false
   )
  ]
  particleSpawnScript [
   /Niagara/Modules/Spawn/Initialization/InitializeParticle(
    Color=<rgba 0.3615 0.3615 0.3615 1>
    Lifetime=&UniformRangedFloat(Minimum=0.2, Maximum=0.5)
    Mass=&RandomRangeFloat(Minimum=0.5, Maximum=1.5)
    Module.Write Color=true
    Module.Write Lifetime=true
    Module.Write Mass=true
    Module.Write Material Random=true
    Module.Write Position=false
    Module.Write Scale=false
    Module.Write SpriteRotation=false
    Module.Write SpriteSize=true
    Position=&SimulationPosition()
    Sprite Size=&Vector2DFromFloat(Value=&FloatFromCurve(
      FloatCurve=DataInterface() {propertyValues {Curve {keys [<at 0.1 1>,<at 0.2 0.5>,<at 1.0011 0.5>]}}}
      CurveIndex=&UniformRangedFloat(Minimum=0, Maximum=360)
      Scale Curve=/Game/Effects/Niagara/DynamicInputs/Scalability/FN_ScalabilityFloatModifier(
       Value at DistanceFraction 0:NiagaraFloat=9
       Value at DistanceFraction 1:NiagaraFloat=15
       DistanceRamp:NiagaraDataInterfaceCurve=DataInterface() {propertyValues {Curve {keys [{interpMode=RCIM_Cubic, tangentMode=RCTM_User, time=0, value=0}, <at 1 1>]}}}
      )
     ))
   ) {enabled=false}
   /Niagara/Modules/Spawn/Initialization/V2/InitializeParticle(
    Color=<rgba 0.3615 0.3615 0.3615 1>
    Lifetime=&UniformRangedFloat(Minimum=0.2, Maximum=0.5)
    Sprite Size Mode=::NewEnumerator3
    Uniform Sprite Size=&MakeFloatFromVector2d(Vector 2D=&Vector2DFromFloat(Value=&FloatFromCurve(
       FloatCurve=DataInterface() {propertyValues {Curve {keys [<at 0.1 1>,<at 0.2 0.5>,<at 1.0011 0.5>]}}}
       CurveIndex=&UniformRangedFloat(Minimum=0, Maximum=360)
       Scale Curve=&FN_ScalabilityFloatModifier(
        Value at DistanceFraction 0:NiagaraFloat=9
        Value at DistanceFraction 1:NiagaraFloat=15
        DistanceRamp:NiagaraDataInterfaceCurve=DataInterface() {propertyValues {Curve {keys [{interpMode=RCIM_Cubic, tangentMode=RCTM_User, time=0, value=0}, <at 1 1>]}}}
       )
      )))
    UsePositionOffset=false
    Write Material Random=true
   ) {module=InitializeParticle001}
   &CylinderLocation(
    Cylinder Axis:Vector3f=<0 0 1>
    Cylinder Height=7.5
    Cylinder Radius=&Multiply_Float(A=30, B=$User.Scale)
    MaskSpawnGroup=false
    Offset=<0 0 5>
    Offset Mode:ENiagara_OffsetMode=::NewEnumerator0
    Override Seed=false
    Rotation Mode:ENiagara_RotationMode=::NewEnumerator1
    UseOwnerScale=true
   )
   &AddVelocity(
    Cone Angle=20
    Cone Axis=<0 0 1>
    Distribution Along Cone Axis=0.3
    Module.UseConeFalloff=true
    Module.UseVelDistribution=true
    Override Seed=false
    Speed Falloff From Cone Axis=0.5
    Velocity Mode=::NewEnumerator2
    Velocity Speed=&RandomRangeFloat(Minimum=250, Maximum=400)
   )
   &AddVelocity(
    Module.Use Falloff Distance=false
    Origin Offset=<0 0 10>
    Override Seed=false
    Velocity Mode=::NewEnumerator1
    Velocity Speed=&RandomRangeFloat(Minimum=100, Maximum=200)
   ) {module=AddVelocity001}
   /Niagara/Modules/Update/SubUV/SubUVAnimation(Number Of Frames=25.6365, Random SubUV Lookup Index:NiagaraFloat=$Particles.Lifetime, SubUV Animation Mode=::NewEnumerator1) {module=SubUVAnimation001}
  ]
  particleUpdateScript [
   &ParticleState()
   &Drag(Use Linear Drag=true, Use Rotational Drag=false)
   &GravityForce()
   &SolveForcesAndVelocity(Clamp Velocity=false, Limit Acceleration=false)
   &ScaleColor(
    Scale Alpha=&FloatFromCurve(FloatCurve=DataInterface() {propertyValues {Curve {keys [
         {arriveTangent=3.5481, time=0, value=1}
         {arriveTangent=-0.0096, interpMode=RCIM_Cubic, leaveTangent=-0.0096, tangentMode=RCTM_User, time=0.5, value=1}
         {arriveTangent=-7.6735, interpMode=RCIM_Cubic, leaveTangent=-7.6735, tangentMode=RCTM_User, time=0.7, value=0}
        ]}}})
    Scale RGB=<0.8 0.8 0.8>
    ScaleA=true
    ScaleRGB=true
   )
   &ScaleSpriteSize(Scale Factor=&Vector2DFromFloat(Value=&FloatFromCurve(
      FloatCurve=DataInterface() {propertyValues {Curve {keys [<at 0.5 0.9191>,<at 0.7 0.002>]}}}
      Scale Curve=&RandomRangeFloat(Minimum=1, Maximum=1.2)
     )))
   &ScaleSpriteSizeBySpeed(Max Scale Factor=<1 1.5>, Min Scale Factor=<0.7 0.7>, Velocity Threshold=250)
  ]
  rendererClasses [&NiagaraSpriteRendererProperties]
  renderers [{rendererClass=&NiagaraSpriteRendererProperties}]
 }
 Dust {
  bEnabled=false
  emitterUpdateScript [
   &EmitterState(Life Cycle Mode=::NewEnumerator1, Loop Behavior=::NewEnumerator1, Loop Delay=0.5, Loop Duration=0.5, UseLoopDelay=false)
   &SpawnBurst_Instantaneous(Spawn Count=10, Spawn Group=1, Use Spawn Probability=false, UseLoopCountLimit=false) {module=SpawnBurst_Instantaneous001}
   &SpawnBurst_Instantaneous(Spawn Count=8, Use Spawn Probability=false, UseLoopCountLimit=false)
  ]
  particleSpawnScript [
   &InitializeParticle(
    Color=&FNTimeOfDayColor(
     Night Color=<rgba 0.2656 0.2656 0.2656 1>
     Morning Color=<rgba 0.2656 0.2656 0.2656 0.2>
     Day Color=<rgba 0.2656 0.2656 0.2656 1>
     Evening Color=<rgba 0.2656 0.2656 0.2656 0.2>
    )
    Lifetime=&UniformRangedFloat(Minimum=0.37, Maximum=0.6)
    Position Offset=<0 0 30>
    Sprite Rotation Mode=::NewEnumerator2
    Sprite Size Mode=::NewEnumerator1
    Uniform Sprite Size Max=120
    Uniform Sprite Size Min=50
    UsePositionOffset=true
    Write Material Random=true
   )
   &DynamicMaterialParameters(
    Index 0 Param 1=&RandomRangeFloat(Minimum=-1, Maximum=1)
    Index 0 Param 2=&RandomRangeFloat(Minimum=0, Maximum=32)
    Module.Index0Float0Write=true
    Module.Index0Float1Write=true
    Module.Index0Float2Write=true
    Module.Index0Float3Write=true
   )
   &CameraOffset(Camera Offset Amount=5)
   &ShapeLocation(MaskSpawnGroup=false, Override Seed=false, Sphere Radius=25, Sphere Surface Distribution=1, UseOwnerScale=true)
   &AddVelocity(
    Module.Use Falloff Distance=false
    Origin Offset=<0 0 -5>
    Origin Offset Coordinate Space=::World
    Override Seed=false
    Velocity Mode=::NewEnumerator1
    Velocity Speed=&RandomRangeFloat(Minimum=100, Maximum=125)
   )
  ]
  particleUpdateScript [
   &ParticleState()
   &Drag(Drag=&RandomRangeFloat(Minimum=0, Maximum=5), Use Linear Drag=true, Use Rotational Drag=true)
   &GravityForce(Gravity=&RandomRangeVector(Minimum=<0 0 -650>, Maximum=<0 0 -980>))
   &SolveForcesAndVelocity(Clamp Velocity=false, Limit Acceleration=false)
   &ScaleColor(
    Scale Alpha=&FloatFromCurve(
     FloatCurve=DataInterface() {propertyValues {Curve {keys [
         {interpMode=RCIM_Cubic, tangentMode=RCTM_User, time=0, value=0}
         <at 0.1 1 RCIM_Cubic>
         {interpMode=RCIM_Cubic, leaveTangent=-3.8631, time=0.75, value=0}
        ]}}}
     Scale Curve=&Multiply_Float(
      A=&FNTimeOfDayScalar(Night Scale=0.7, Morning Scale=0.8, Day Scale=1.5, Evening Scale=0.8)
      B=0.3
     )
    )
    ScaleA=true
    ScaleRGB=true
   ) {enabled=false, module=ScaleColor001}
   &ScaleSpriteSize(Uniform Curve Sprite Scale=DataInterface() {propertyValues {Curve {keys [
        {arriveTangent=0.6664, interpMode=RCIM_Cubic, leaveTangent=0.6664, tangentMode=RCTM_User, time=0, value=0.777}
        {interpMode=RCIM_Cubic, leaveTangent=-0.4663, time=1, value=1.1262}
       ]}}})
   /SetVariables(Particles.Dust color:LinearColor=&MakeLinearColorFromVectorAndFloat(
     "Vector (RGB)"=&VectorFromFloat(Value=&FloatFromCurve(FloatCurve=DataInterface() {propertyValues {Curve {keys [<at 0 1>,<at 0.9983 0.5>]}}}, Scale Curve=15))
     "Float (Alpha)"=&Multiply_Float(
      A=&FloatFromCurve(FloatCurve=DataInterface() {propertyValues {Curve {keys [<at 0 0>,<at 0.2049 0.9765>,<at 0.4942 0.9474>,<at 0.7 0>]}}})
      B=&Multiply_Float(A=&FNTimeOfDayScalar(Night Scale=1, Morning Scale=0.3, Day Scale=1.5, Evening Scale=0.3), B=1)
     )
    )) {isSetParameters=true}
   /SetVariables(Particles.smoke size:Vector2f=&Vector2DFromFloat(Value=&RandomRangeFloat(Minimum=75, Maximum=85))) {isSetParameters=true, module=SetVariables_#1}
  ]
  rendererClasses [&NiagaraSpriteRendererProperties]
  renderers [
   {rendererClass=&NiagaraSpriteRendererProperties}
   {rendererClass=&NiagaraSpriteRendererProperties, rendererIndex=1}
   {rendererClass=&NiagaraSpriteRendererProperties, rendererIndex=2}
  ]
 }
 Dust001 {
  emitterUpdateScript [
   &EmitterState(Life Cycle Mode=::NewEnumerator1, Loop Behavior=::NewEnumerator1, Loop Delay=0.5, Loop Duration=35, UseLoopDelay=false)
   &SpawnBurst_Instantaneous(
    Spawn Count=&MakeIntFromFloat(Float=&FN_FrontEndFloatModifier(Gameplay=7, Front End=0))
    Use Spawn Probability=false
    UseLoopCountLimit=false
   )
  ]
  particleSpawnScript [
   &InitializeParticle(
    Color=&FNTimeOfDayColor(
     Night Color=&RandomRangeLinearColor(Minimum=<rgba 0.2198 0.2198 0.2198 0.1>, Maximum=<rgba 0.2656 0.2656 0.2656 0.2>)
     Morning Color=<rgba 0.2656 0.2656 0.2656 0.2>
     Day Color=<rgba 0.2656 0.2656 0.2656 0.2>
     Evening Color=<rgba 0.2656 0.2656 0.2656 0.2>
    )
    Lifetime=&UniformRangedFloat(Minimum=0.2, Maximum=0.5)
    Position Offset=<0 0 30>
    Sprite Rotation Mode=::NewEnumerator2
    Sprite Size Mode=::NewEnumerator1
    Uniform Sprite Size Max=85
    Uniform Sprite Size Min=50
    UsePositionOffset=true
    Write Material Random=true
   )
   &CameraOffset(Camera Offset Amount=5)
   &ShapeLocation(MaskSpawnGroup=false, Override Seed=false, Sphere Radius=10, Sphere Surface Distribution=1, UseOwnerScale=true)
   &AddVelocity(
    Module.Use Falloff Distance=false
    Origin Offset Coordinate Space=::World
    Override Seed=false
    Velocity Mode=::NewEnumerator1
    Velocity Speed=&RandomRangeFloat(Minimum=100, Maximum=125)
   )
  ]
  particleUpdateScript [
   &ParticleState()
   &Drag(Drag=&RandomRangeFloat(Minimum=0, Maximum=5), Use Linear Drag=true, Use Rotational Drag=true)
   &GravityForce(Gravity=&RandomRangeVector(Minimum=<0 0 -200>, Maximum=<0 0 -980>))
   &AccelerationForce(Acceleration=&MakeVector(X=0, Y=0, Z=&UniformRangedFloat(Minimum=0, Maximum=30)))
   &SolveForcesAndVelocity(Clamp Velocity=false, Limit Acceleration=false)
   &ScaleSpriteSize(Uniform Curve Sprite Scale=DataInterface() {propertyValues {Curve {keys [
        {arriveTangent=0.6664, interpMode=RCIM_Cubic, leaveTangent=0.6664, tangentMode=RCTM_User, time=0, value=0.777}
        {interpMode=RCIM_Cubic, leaveTangent=-0.4663, time=1, value=1}
       ]}}})
   /SetVariables(Particles.smoke size:Vector2f=&Vector2DFromFloat(Value=&FloatFromCurve(FloatCurve=DataInterface() {propertyValues {Curve {keys [
          {arriveTangent=7.1593, interpMode=RCIM_Cubic, leaveTangent=7.1593, tangentMode=RCTM_User, time=0, value=0}
          {arriveTangent=1.1699, interpMode=RCIM_Cubic, leaveTangent=1.1699, tangentMode=RCTM_User, time=0.31, value=0.97}
          <at 1 1.25 RCIM_Cubic>
         ]}}}))) {isSetParameters=true, module=SetVariables_#1}
   &DynamicMaterialParameters(
    Index 0 Param 1=&FloatFromCurve(FloatCurve=DataInterface() {propertyValues {Curve {keys [
         {arriveTangent=0.9088, interpMode=RCIM_Cubic, leaveTangent=0.9088, tangentMode=RCTM_User, time=0, value=0.5}
         {arriveTangent=0.5473, interpMode=RCIM_Cubic, leaveTangent=0.5473, tangentMode=RCTM_User, time=1, value=1}
        ]}}})
    Index 0 Param 3=&FloatFromCurve(FloatCurve=DataInterface() {propertyValues {Curve {keys [
         {arriveTangent=-9.0447, interpMode=RCIM_Cubic, leaveTangent=-2.0044, tangentMode=RCTM_User, time=0, value=0.5}
         {interpMode=RCIM_Cubic, leaveTangent=4.6779, time=0.2851, value=0}
        ]}}})
    Index 1 Param 1=&FloatFromCurve(
     FloatCurve=DataInterface() {propertyValues {Curve {keys [<at 0 0>,<at 0.25 1>]}}}
     Scale Curve=&RandomRangeFloat(Minimum=2, Maximum=1)
    )
    Index 1 Param 2=&FloatFromCurve(
     FloatCurve=DataInterface() {propertyValues {Curve {keys [
         {interpMode=RCIM_Cubic, leaveTangent=3.7422, tangentMode=RCTM_Break, time=0, value=0.2}
         {interpMode=RCIM_Cubic, leaveTangent=0.9066, time=1, value=2}
        ]}}}
     Scale Curve=&RandomRangeFloat(Minimum=0.5, Maximum=1)
    )
    Index 1 Param 3=&FloatFromCurve(FloatCurve=DataInterface() {propertyValues {Curve {keys [
         {interpMode=RCIM_Cubic, leaveTangent=2.8445, tangentMode=RCTM_Break, time=0, value=0.2}
         {interpMode=RCIM_Cubic, leaveTangent=63.2978, time=0.25, value=1}
        ]}}})
    Index 1 Param 4=&FloatFromCurve(
     FloatCurve=DataInterface() {propertyValues {Curve {keys [<at 0 1>,<at 1 0.8>]}}}
     Scale Curve=&RandomRangeFloat(Minimum=0.5, Maximum=1)
    )
    Index 2 Param 1=&FloatFromCurve(FloatCurve=DataInterface() {propertyValues {Curve {keys [{interpMode=RCIM_Cubic, leaveTangent=3.2309, tangentMode=RCTM_Break, time=0.1, value=0}, <at 1 1>]}}})
    Index 2 Param 2=&FloatFromCurve(
     FloatCurve=DataInterface() {propertyValues {Curve {keys [<at 0 0.2 RCIM_Cubic>,<at 1 0.8 RCIM_Cubic>]}}}
     Scale Curve=&RandomRangeFloat(Minimum=0.2, Maximum=1.2)
    )
    Index 2 Param 3=&FloatFromCurve(FloatCurve=DataInterface() {propertyValues {Curve {keys [
         {interpMode=RCIM_Cubic, tangentMode=RCTM_Break, time=0, value=1}
         {interpMode=RCIM_Cubic, leaveTangent=11.1233, tangentMode=RCTM_Break, time=1, value=2}
        ]}}})
    Index 2 Param 4=&FloatFromCurve(
     FloatCurve=DataInterface() {propertyValues {Curve {keys [<at 0 1 RCIM_Cubic>, {interpMode=RCIM_Cubic, leaveTangent=-1.2956, tangentMode=RCTM_Break, time=1, value=0.75}]}}}
     Scale Curve=0.8
    )
    Module.Index0Float0Write=true
    Module.Index0Float1Write=true
    Module.Index0Float2Write=true
    Module.Index0Float3Write=true
    Module.Index1Float0Write=true
    Module.Index1Float1Write=true
    Module.Index1Float2Write=true
    Module.Index1Float3Write=true
    Module.Index2Float0Write=true
    Module.Index2Float1Write=true
    Module.Index2Float2Write=true
    Module.Index2Float3Write=true
    Write Parameter Index 1=::NewEnumerator1
    Write Parameter Index 2=::NewEnumerator1
   )
   &ScaleColor(
    Scale Alpha=&Multiply_Float(
     A=&FloatFromCurve(
      FloatCurve=DataInterface() {propertyValues {Curve {keys [<at 0 0>,<at 0.2049 0.9765>,<at 0.7791 0.9107>,<at 1 0>]}}}
      Scale Curve=&FN_FrontEndFloatModifier(Gameplay=1, Front End=0.2)
     )
     B=&Multiply_Float(
      A=&FNTimeOfDayScalar(Night Scale=0.7, Morning Scale=0.5, Evening Scale=0.3)
      B=&FN_FrontEndFloatModifier(Gameplay=1, Front End=0.2)
     )
    )
    Scale RGB=&VectorFromFloat(Value=&FloatFromCurve(FloatCurve=DataInterface() {propertyValues {Curve {keys [<at 0 1>,<at 0.9983 0.5>]}}}))
    ScaleA=true
    ScaleRGB=true
   )
  ]
  rendererClasses [&NiagaraSpriteRendererProperties]
  renderers [
   {rendererClass=&NiagaraSpriteRendererProperties}
   {rendererClass=&NiagaraSpriteRendererProperties, rendererIndex=1}
  ]
 }
 LWE_SMOKE {bEnabled=false, rendererClasses [], renderers []}
}
summary {
 systemName=NS_ClaimReflect_PropDespawn_rockdust_WorldSpace
 emitters [
  {bEnabled=false, emitterName=Dust, rendererClasses [&NiagaraSpriteRendererProperties]}
  {bEnabled=false, emitterName=Dirt, rendererClasses [&NiagaraSpriteRendererProperties]}
  {emitterName=Dust001, rendererClasses [&NiagaraSpriteRendererProperties]}
  {bEnabled=false, emitterName=LWE_SMOKE, rendererClasses []}
 ]
 userVariables [
  User.bIsFrontEnd {type=NiagaraBool}
  User.bIsFrontEndPreview {type=NiagaraBool}
  User.NormalizedNotifyProgress {type=NiagaraFloat, defaultValue=0, description=""}
  User.Scale {type=NiagaraFloat, defaultValue=1, description=""}
 ]
}
