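# Niagara impact effect for the ClaimReflect emote's rock hit; only the directional smoke emitter is enabled, producing a short dust puff.
# It bursts sprites from a small torus, pushes them outward, then slows them with drag and gravity.
# Curves drive size, fade and dynamic material parameters, with time-of-day alpha scaling and a front-end modifier that zeroes the spawn count.
# Rock-mesh debris and two other smoke emitters are disabled, including the only logic reading the velocity user parameters.
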NiagaraSystem /BRCosmetics/Animation/Game/MainPlayer/Emotes/ClaimReflect/FX/NS_Emote_ClaimReflect_RockImpact
dependencies {
 usedDataInterfaces [&NiagaraDataInterfaceCurve, &NiagaraDataInterfaceColorCurve]
 usedDynamicInputs [&RotateVectorByQuaternion, &MakeQuatFromAxisAngle]
 usedModules [&SystemState]
}
emitters {
 LWE_SMOKE {bEnabled=false, rendererClasses [], renderers []}
 Rocks_Burst {
  bEnabled=false
  emitterUpdateScript [
   &EmitterState(Life Cycle Mode=::NewEnumerator1, Loop Behavior=::NewEnumerator1, Loop Duration=1, UseLoopDelay=true)
   &SpawnBurst_Instantaneous(
    Spawn Count=&RandomRangeInt(Minimum=7, Maximum=12)
    Use Spawn Probability=false
    UseLoopCountLimit=false
   )
   /SetVariables() {isSetParameters=true}
  ]
  particleSpawnScript [
   &InitializeParticle(
    Lifetime Max=0.8
    Lifetime Min=0.6
    Lifetime Mode=::NewEnumerator1
    Mesh Scale Mode=::NewEnumerator3
    Mesh Uniform Scale=&Multiply_Float(
     A=&Lerp_Float(
      A=0.0285
      B=0.06
      Alpha=&FloatFromCurve(
       FloatCurve=DataInterface() {propertyValues {Curve {keys [<at 0 0 RCIM_Constant>, {arriveTangent=7.6187, interpMode=RCIM_Constant, leaveTangent=7.6187, time=0.65, value=1}]}}}
       CurveIndex=&RandomRangeFloat(Minimum=0, Maximum=1)
      )
     )
     B=1.4
    )
    UsePositionOffset=false
    Write Material Random=true
   )
   &ShapeLocation(
    Handle Radius=0
    Large Radius=2
    MaskSpawnGroup=false
    Offset=<0 0 7>
    Offset Mode=::NewEnumerator0
    Override Seed=false
    Shape Primitive=::NewEnumerator3
    UseOwnerScale=true
   )
   &InitialMeshOrientation(Use Rotation Vector=false)
   &AddRotationalVelocity(Rotation Rate=&Multiply_VectorByFloat(
     Vector=&RandomVector(Vector Scale=2)
     Float=&RandomRangeFloat(Minimum=0.5, Maximum=1)
    ))
   &DynamicMaterialParameters(
    Index 0 Param 2=1
    Module.Param0WriteEnabled=false
    Module.Param1WriteEnabled=true
    Module.Param2WriteEnabled=false
    Module.Param3WriteEnabled=false
   )
   &AddVelocity(
    Rotation Coordinate Space=::World
    Velocity=&MakeVector(
     X=&RandomRangeFloat(Minimum=0, Maximum=$User.Anim_Curves_VelX)
     Y=&RandomRangeFloat(Minimum=-100, Maximum=$User.Anim_Curves_VelY)
     Z=&RandomRangeFloat(Minimum=150, Maximum=350)
    )
   )
   &AddVelocity(
    Module.Use Falloff Distance=false
    Override Seed=false
    Velocity Mode=::NewEnumerator1
    Velocity Speed=&RandomRangeFloat(Minimum=50, Maximum=225)
   ) {module=AddVelocity001}
   &ApplyInitialForces()
   /SetVariables(Particles.MeshIndex=&RandomRangeInt(Minimum=0, Maximum=2)) {isSetParameters=true}
  ]
  particleUpdateScript [
   &ParticleState()
   &GravityForce(Gravity=&RandomRangeVector(Minimum=<0 0 -750>, Maximum=<0 0 -1500>))
   &Collision(
    Age Colliding Particles=false
    KillOnCollision=false
    Module.Control Rest State Via Surface Normal=false
    Module.EnableMaxCollisionCount=false
    Module.Randomize Collision Normal Vector=false
    Restitution=0.2
   )
   &Drag(
    Drag=1.5
    Rotational Drag:NiagaraFloat=&FloatFromCurve(FloatCurve=DataInterface() {propertyValues {Curve {keys [<at 0 0 RCIM_Cubic>, {arriveTangent=0.1846, interpMode=RCIM_Cubic, leaveTangent=0.1846, tangentMode=RCTM_User, time=1, value=0.1}]}}})
    Use Linear Drag=true
    Use Rotational Drag=true
   )
   &SolveForcesAndVelocity(Clamp Velocity=false, Limit Acceleration=false)
   &ScaleMeshSize(Scale Factor=&VectorFromFloat(Value=&FloatFromCurve(FloatCurve=DataInterface() {propertyValues {Curve {keys [{arriveTangent=-0.0035, interpMode=RCIM_Cubic, leaveTangent=-0.0035, tangentMode=RCTM_User, time=0.6, value=1}, <at 1 0>]}}})))
  ]
  rendererClasses [&NiagaraMeshRendererProperties]
  renderers [{rendererClass=&NiagaraMeshRendererProperties, rendererIndex=0}]
 }
 Smoke {
  bEnabled=false
  emitterUpdateScript [
   &EmitterState()
   &SpawnBurst_Instantaneous(Spawn Count=7, Use Spawn Probability=false) {module=SpawnBurst_Instantaneous001}
  ]
  particleSpawnScript [
   &InitializeParticle(
    Color=<rgba 0.6948 0.6468 0.4891 1>
    Lifetime=&UniformRangedFloat(Minimum=0.6, Maximum=0.7)
    Sprite Rotation Angle=&UniformRangedFloat(Minimum=0, Maximum=1)
    Sprite Rotation Mode=::NewEnumerator1
    Sprite Size Mode=::NewEnumerator3
    Uniform Sprite Size=&UniformRangedFloat(Minimum=25, Maximum=35)
    UsePositionOffset=false
    Write Material Random=true
   ) {module=InitializeParticle001}
   /SetVariables(Particles.InitSpriteRotRate=&UniformRangedFloat(Minimum=-1, Maximum=1)) {isSetParameters=true}
   &AccelerationForce(Acceleration=&UniformRangedVector(Minimum=<0 0 15>, Maximum=<0 0 25>))
   &SphereLocation(
    MaskSpawnGroup=false
    Offset Coordinate Space:ENiagaraCoordinateSpace=::World
    Override Seed=false
    Sphere Radius=10
    Surface Only=true
    UseNonUniformScale=false
    UseOffset=true
    UseSphereOrientation=false
   )
   &ApplyInitialForces()
   &CameraOffset(Camera Offset Amount=&RandomRangeFloat(Minimum=0, Maximum=15))
   &AddVelocity(Velocity=&Multiply_Vector(
     A=$Output.SphereLocation.SphereNormal
     B=&VectorFromFloat(Value=&RandomRangeFloat(Minimum=20, Maximum=30))
    ))
  ]
  particleUpdateScript [
   &ParticleState()
   &ScaleColor(
    Scale Mode=::NewEnumerator1
    Scale RGBA=&MakeVector4_FromLinearColor(LinearColor=&FNTimeOfDayColor(
      Night Color=&RandomRangeLinearColor(Minimum=<rgba 0.478 0.478 0.6313 1>, Maximum=<rgba 0.781 0.781 1 1>)
      Morning Color=&RandomRangeLinearColor(Minimum=<rgba 0.2813 0.2326 0.478 1>, Maximum=<rgba 0.3377 0.5872 0.9253 1>)
      Day Color=&RandomRangeLinearColor(Minimum=<rgba 0.478 0.4265 0.4121 1>, Maximum=<rgba 0.781 0.718 0.6943 1>)
      Evening Color=&RandomRangeLinearColor(Minimum=<rgba 0.478 0.478 0.478 1>, Maximum=<rgba 0.781 0.781 0.781 1>)
     ))
    ScaleRGBA=true
   ) {module=ScaleColor001}
   &ScaleColor(
    Scale Alpha=&Multiply_Float(
     A=&FloatFromCurve(
      FloatCurve=DataInterface() {propertyValues {Curve {keys [
          {interpMode=RCIM_Cubic, tangentMode=RCTM_User, time=0, value=0.55}
          {arriveTangent=-0.6111, interpMode=RCIM_Cubic, leaveTangent=-0.6111, time=0.15, value=1}
          <at 0.9 0>
         ]}}}
      Scale Curve=&FN_FrontEndFloatModifier(Gameplay=0.75, Front End=0.3)
     )
     B=&FloatFromCurve(
      FloatCurve=DataInterface() {propertyValues {Curve {keys [
          {interpMode=RCIM_Cubic, tangentMode=RCTM_User, time=0, value=0.6}
          {interpMode=RCIM_Cubic, tangentMode=RCTM_User, time=0.5, value=1}
          {tangentMode=RCTM_User, time=1, value=0.6}
         ]}}}
      CurveIndex=Linked() {linkedVariable {name=NPC.FortniteNPC.TimeOfDay_0-1, type=NiagaraFloat}}
     )
    )
    ScaleA=true
    ScaleRGB=false
   )
   &SpriteRotationRate(Rotation Rate=&Multiply_Float(
     A=$Particles.InitSpriteRotRate
     B=&FloatFromCurve(
      FloatCurve=DataInterface() {propertyValues {Curve {keys [
          {arriveTangent=-1.9214, interpMode=RCIM_Cubic, leaveTangent=-1.9214, tangentMode=RCTM_User, time=-0.0046, value=1.0113}
          {arriveTangent=-0.2496, interpMode=RCIM_Cubic, leaveTangent=-0.2496, tangentMode=RCTM_User, time=0.9896, value=0.16}
         ]}}}
      CurveIndex=$Emitter.Age
     )
    ))
   &Drag(Drag=3, Use Linear Drag=true, Use Rotational Drag=false)
   &SolveForcesAndVelocity(Clamp Velocity=false, Limit Acceleration=false)
   &ScaleSpriteSize(Scale Factor=&Vector2DFromFloat(Value=&FloatFromCurve(FloatCurve=DataInterface() {propertyValues {Curve {keys [
          {arriveTangent=5.2354, interpMode=RCIM_Cubic, leaveTangent=5.2354, tangentMode=RCTM_User, time=0, value=1.25}
          <at 0.1439 1.4923>
          <at 1.0011 3.4809 RCIM_Cubic>
         ]}}})))
   &DynamicMaterialParameters(
    Index 0 Param 1=&FloatFromCurve(FloatCurve=DataInterface() {propertyValues {Curve {keys [<at 0 0 RCIM_Cubic>,<at 1.0011 0.2>]}}})
    Index 0 Param 2=&FloatFromCurve(FloatCurve=DataInterface() {propertyValues {Curve {keys [
         {arriveTangent=0.0453, interpMode=RCIM_Cubic, tangentMode=RCTM_Break, time=0, value=-1}
         {arriveTangent=0.0604, interpMode=RCIM_Cubic, leaveTangent=0.0604, tangentMode=RCTM_User, time=1, value=-1}
        ]}}})
    Index 0 Param 3=&FloatFromCurve(FloatCurve=DataInterface() {propertyValues {Curve {keys [{interpMode=RCIM_Cubic, leaveTangent=4.6779, time=0.2851, value=0}]}}})
    Index 1 Param 1=&FloatFromCurve(
     FloatCurve=DataInterface() {propertyValues {Curve {keys [<at 0 0>,<at 0.25 1>]}}}
     Scale Curve=&RandomRangeFloat(Minimum=2, Maximum=1)
    )
    Index 1 Param 2=&FloatFromCurve(
     FloatCurve=DataInterface() {propertyValues {Curve {keys [
         {interpMode=RCIM_Cubic, leaveTangent=3.7422, tangentMode=RCTM_Break, time=0, value=0.2}
         {interpMode=RCIM_Cubic, leaveTangent=0.9066, time=1, value=2}
        ]}}}
     Scale Curve=&RandomRangeFloat(Minimum=0.5, Maximum=1)
    )
    Index 1 Param 3=&FloatFromCurve(FloatCurve=DataInterface() {propertyValues {Curve {keys [
         {interpMode=RCIM_Cubic, leaveTangent=2.8445, tangentMode=RCTM_Break, time=0, value=0.2}
         {interpMode=RCIM_Cubic, leaveTangent=63.2978, time=0.25, value=1}
        ]}}})
    Index 1 Param 4=&FloatFromCurve(
     FloatCurve=DataInterface() {propertyValues {Curve {keys [<at 0 1>,<at 1 0.8>]}}}
     Scale Curve=&RandomRangeFloat(Minimum=0.5, Maximum=1)
    )
    Index 2 Param 1=&FloatFromCurve(FloatCurve=DataInterface() {propertyValues {Curve {keys [{interpMode=RCIM_Cubic, leaveTangent=3.2309, tangentMode=RCTM_Break, time=0.1, value=0}, <at 1 1>]}}})
    Index 2 Param 2=$Particles.MaterialRandom
    Index 2 Param 3=&FloatFromCurve(FloatCurve=DataInterface() {propertyValues {Curve {keys [
         {interpMode=RCIM_Cubic, tangentMode=RCTM_Break, time=0, value=1}
         {interpMode=RCIM_Cubic, leaveTangent=11.1233, tangentMode=RCTM_Break, time=1, value=2}
        ]}}})
    Index 2 Param 4=0.75
    Module.Param0WriteEnabled=true
    Module.Param0WriteEnabled1=true
    Module.Param0WriteEnabled2=true
    Module.Param1WriteEnabled=false
    Module.Param1WriteEnabled1=true
    Module.Param1WriteEnabled2=true
    Module.Param2WriteEnabled=true
    Module.Param2WriteEnabled1=true
    Module.Param2WriteEnabled2=true
    Module.Param3WriteEnabled=true
    Module.Param3WriteEnabled1=true
    Module.Param3WriteEnabled2=true
    Write Parameter Index 1=true
    Write Parameter Index 2=true
   )
  ]
  rendererClasses [&NiagaraSpriteRendererProperties]
  renderers [{rendererClass=&NiagaraSpriteRendererProperties}]
 }
 Smoke_directional_with_curve {
  emitterUpdateScript [
   &EmitterState(Life Cycle Mode=::NewEnumerator1, Loop Behavior=::NewEnumerator1, Loop Duration=0.5, UseLoopDelay=false)
   &SpawnBurst_Instantaneous(
    Spawn Count=&MakeIntFromFloat(Float=&FN_FrontEndFloatModifier(Gameplay=7, Front End=0))
    Use Spawn Probability=false
   ) {module=SpawnBurst_Instantaneous001}
  ]
  particleSpawnScript [
   &InitializeParticle(
    Color=<rgba 0.2656 0.2656 0.2656 0.2>
    Lifetime Max=0.5
    Lifetime Min=0.3
    Lifetime Mode=::NewEnumerator1
    Sprite Rotation Mode=::NewEnumerator2
    Sprite Size Mode=::NewEnumerator1
    Uniform Sprite Size Max=55
    Uniform Sprite Size Min=30
    UsePositionOffset=false
    Write Material Random=true
   ) {module=InitializeParticle001}
   &TorusLocation(
    Align Torus With Axis=true
    Coordinate Space:ENiagaraCoordinateSpace=::World
    Large Radius=2
    MaskSpawnGroup=false
    Module.Override Randomness=false
    Offset=<0 0 1>
    Torus Mode=::NewEnumerator2
   )
   &AddVelocityFromPoint(
    Module.Use Falloff Distance=false
    Origin Offset=<0 0 2>
    Origin Offset Coordinate Space=::World
    Velocity Origin:Vector3f=&SimulationPosition()
    Velocity Strength=&UniformRangedFloat(Minimum=64, Maximum=240)
   )
   &AddVelocity(
    Module.Use Falloff Distance=false
    Override Seed=false
    Velocity Mode=::NewEnumerator1
    Velocity Speed=&RandomRangeFloat(Minimum=0, Maximum=64)
   ) {module=AddVelocity001}
   &ApplyInitialForces()
   &AddVelocity(Coordinate Space=::World, Velocity=&RandomRangeVector(Minimum=<0 0 0>, Maximum=<0 0 40>)) {enabled=false}
   &AddVelocity(Velocity=&Multiply_Vector(
     A=$Output.TorusLocation.TorusNormal
     B=&VectorFromFloat(Value=&RandomRangeFloat(Minimum=20, Maximum=30))
    )) {module=AddVelocity002}
  ]
  particleUpdateScript [
   &ParticleState()
   &AccelerationForce(Acceleration=&MakeVector(X=0, Y=0, Z=&UniformRangedFloat(Minimum=0, Maximum=30)))
   &Drag(Drag=5, Use Linear Drag=true, Use Rotational Drag=false)
   &GravityForce(Gravity=<0 0 -250>)
   &SolveForcesAndVelocity(Clamp Velocity=false, Limit Acceleration=false)
   &ScaleSpriteSize(Scale Factor=&Vector2DFromFloat(Value=&FloatFromCurve(FloatCurve=DataInterface() {propertyValues {Curve {keys [
          {arriveTangent=7.1593, interpMode=RCIM_Cubic, leaveTangent=7.1593, tangentMode=RCTM_User, time=0, value=0}
          {arriveTangent=1.1699, interpMode=RCIM_Cubic, leaveTangent=1.1699, tangentMode=RCTM_User, time=0.31, value=0.97}
          <at 1 1.25 RCIM_Cubic>
         ]}}})))
   &DynamicMaterialParameters(
    Index 0 Param 1=&FloatFromCurve(FloatCurve=DataInterface() {propertyValues {Curve {keys [
         {arriveTangent=0.9088, interpMode=RCIM_Cubic, leaveTangent=0.9088, tangentMode=RCTM_User, time=0, value=1}
         {arriveTangent=0.5473, interpMode=RCIM_Cubic, leaveTangent=0.5473, tangentMode=RCTM_User, time=1, value=2}
        ]}}})
    Index 0 Param 3=&FloatFromCurve(FloatCurve=DataInterface() {propertyValues {Curve {keys [
         {arriveTangent=-9.0447, interpMode=RCIM_Cubic, leaveTangent=-2.0044, tangentMode=RCTM_User, time=0, value=0.5}
         {interpMode=RCIM_Cubic, leaveTangent=4.6779, time=0.2851, value=0}
        ]}}})
    Index 1 Param 1=&FloatFromCurve(
     FloatCurve=DataInterface() {propertyValues {Curve {keys [<at 0 0>,<at 0.25 1>]}}}
     Scale Curve=&RandomRangeFloat(Minimum=2, Maximum=1)
    )
    Index 1 Param 2=&FloatFromCurve(
     FloatCurve=DataInterface() {propertyValues {Curve {keys [
         {interpMode=RCIM_Cubic, leaveTangent=3.7422, tangentMode=RCTM_Break, time=0, value=0.2}
         {interpMode=RCIM_Cubic, leaveTangent=0.9066, time=1, value=2}
        ]}}}
     Scale Curve=&RandomRangeFloat(Minimum=0.5, Maximum=1)
    )
    Index 1 Param 3=&FloatFromCurve(FloatCurve=DataInterface() {propertyValues {Curve {keys [
         {interpMode=RCIM_Cubic, leaveTangent=2.8445, tangentMode=RCTM_Break, time=0, value=0.2}
         {interpMode=RCIM_Cubic, leaveTangent=63.2978, time=0.25, value=1}
        ]}}})
    Index 1 Param 4=&FloatFromCurve(
     FloatCurve=DataInterface() {propertyValues {Curve {keys [<at 0 1>,<at 1 0.8>]}}}
     Scale Curve=&RandomRangeFloat(Minimum=0.5, Maximum=1)
    )
    Index 2 Param 1=&FloatFromCurve(FloatCurve=DataInterface() {propertyValues {Curve {keys [{interpMode=RCIM_Cubic, leaveTangent=3.2309, tangentMode=RCTM_Break, time=0.1, value=0}, <at 1 1>]}}})
    Index 2 Param 2=&FloatFromCurve(
     FloatCurve=DataInterface() {propertyValues {Curve {keys [<at 0 0.2 RCIM_Cubic>,<at 1 0.8 RCIM_Cubic>]}}}
     Scale Curve=&RandomRangeFloat(Minimum=0.2, Maximum=1.2)
    )
    Index 2 Param 3=&FloatFromCurve(FloatCurve=DataInterface() {propertyValues {Curve {keys [
         {interpMode=RCIM_Cubic, tangentMode=RCTM_Break, time=0, value=1}
         {interpMode=RCIM_Cubic, leaveTangent=11.1233, tangentMode=RCTM_Break, time=1, value=2}
        ]}}})
    Index 2 Param 4=&FloatFromCurve(
     FloatCurve=DataInterface() {propertyValues {Curve {keys [<at 0 1 RCIM_Cubic>, {interpMode=RCIM_Cubic, leaveTangent=-1.2956, tangentMode=RCTM_Break, time=1, value=0.75}]}}}
     Scale Curve=0.8
    )
    Module.Param0WriteEnabled=true
    Module.Param0WriteEnabled1=true
    Module.Param0WriteEnabled2=true
    Module.Param1WriteEnabled=true
    Module.Param1WriteEnabled1=true
    Module.Param1WriteEnabled2=true
    Module.Param2WriteEnabled=true
    Module.Param2WriteEnabled1=true
    Module.Param2WriteEnabled2=true
    Module.Param3WriteEnabled=true
    Module.Param3WriteEnabled1=true
    Module.Param3WriteEnabled2=true
    Write Parameter Index 1=true
    Write Parameter Index 2=true
   )
   &ScaleColor(
    Scale Alpha=&Multiply_Float(
     A=&FloatFromCurve(FloatCurve=DataInterface() {propertyValues {Curve {keys [<at 0 0>,<at 0.2049 0.9765>,<at 0.7791 0.9107>,<at 1 0>]}}})
     B=&Multiply_Float(A=&FNTimeOfDayScalar(Night Scale=1, Morning Scale=0.3, Evening Scale=0.3), B=1)
    )
    Scale RGB=&VectorFromFloat(Value=&FloatFromCurve(FloatCurve=DataInterface() {propertyValues {Curve {keys [<at 0 1>,<at 0.9983 0.5>]}}}))
    ScaleA=true
    ScaleRGB=true
   )
  ]
  rendererClasses [&NiagaraSpriteRendererProperties]
  renderers [
   {rendererClass=&NiagaraSpriteRendererProperties}
   {rendererClass=&NiagaraSpriteRendererProperties, rendererIndex=1}
  ]
 }
}
summary {
 systemName=NS_Emote_ClaimReflect_RockImpact
 emitters [
  {emitterName=Smoke_directional_with_curve, rendererClasses [&NiagaraSpriteRendererProperties]}
  {bEnabled=false, emitterName=Rocks_Burst, rendererClasses [&NiagaraMeshRendererProperties]}
  {bEnabled=false, emitterName=Smoke, rendererClasses [&NiagaraSpriteRendererProperties]}
  {bEnabled=false, emitterName=LWE_SMOKE, rendererClasses []}
 ]
 userVariables [
  User.Anim_Curves_VelX {type=NiagaraFloat, defaultValue=0, description=""}
  User.Anim_Curves_VelY {type=NiagaraFloat, defaultValue=100, description=""}
  User.bIsFrontEnd {type=NiagaraBool}
  User.bIsFrontEndPreview {type=NiagaraBool}
  User.OwnerLocalVelocity {type=Vector3f, description="", defaultValue {struct=Vector3f, value=<0 0 0>}}
 ]
}
