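# Niagara system for traversal dust on the ClaimReflect emote, with all five emitters (debris, dust, smoke, ground rings, Test) disabled in this asset.
# Debris, dust and smoke spawn from the skeletal mesh at speed-dependent rates, while ground rings spawn only when the owner is nearly stationary.
# Front-end modifiers swap rates and offsets for menu display, and time-of-day functions tint dust, smoke and rings.
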NiagaraSystem /BRCosmetics/Animation/Game/MainPlayer/Emotes/ClaimReflect/FX/NS_Emote_ClaimReflect_TraversalDust
dependencies {
 usedDataInterfaces [
  &NiagaraDataInterfaceCurve
  &NiagaraDataInterfaceColorCurve
  &NiagaraDataInterfaceVectorCurve
  &NiagaraDataInterfacePlatformSet
  &NiagaraDataInterfaceSkeletalMesh
 ]
 usedDynamicInputs [&UniformRangedVector2D, &MakeFloatFromVector2d]
 usedModules [
  &self:SystemSpawnScript.SystemScriptSource.SystemScriptGraph.NiagaraNodeAssignment_0.SetVariables_#1
  &SystemState
 ]
}
emitters {
 Debris003 {
  bEnabled=false
  emitterUpdateScript [
   &EmitterState()
   &SpawnRate(
    SpawnRate=&FN_FrontEndFloatModifier(
     Gameplay=&MakeCustomFloatFromBool(
      Bool=&SetBoolByFloatComparison(A=&VectorLength(Vector=$User.OwnerLocalVelocity), B=150)
      True Float=30
      False Float=0
     )
     Front End=15
    )
    Use Spawn Probability=false
   )
  ]
  particleSpawnScript [
   &InitializeParticle(
    Color=&ColorFromCurve(
     ColorCurve=DataInterface() {propertyValues {
       AlphaCurve {keys [<at -0.0023 1 RCIM_Cubic>]}
       BlueCurve {keys [<at -0.0023 0.2122 RCIM_Cubic>,<at 0.1383 0.2122>,<at 0.2178 0.1859>,<at 0.4486 0.75>,<at 0.7126 0.1859>,<at 1.0234 0.8802>]}
       GreenCurve {keys [<at -0.0023 0.2325 RCIM_Cubic>,<at 0.1383 0.2325>,<at 0.2178 0.2005>,<at 0.4486 0.8456>,<at 0.7126 0.2122>,<at 1.0234 0.9474>]}
       RedCurve {keys [<at -0.0023 0.25 RCIM_Cubic>,<at 0.1383 0.25>,<at 0.2178 0.2188>,<at 0.4486 1>,<at 0.7126 0.2188>,<at 1.0234 1>]}
      }}
     CurveIndex=&UniformRangedFloat(Minimum=0, Maximum=1)
    )
    Lifetime=&UniformRangedFloat(
     Minimum=&FN_FrontEndFloatModifier(Gameplay=0.5, Front End=0.25)
     Maximum=&FN_FrontEndFloatModifier(Gameplay=0.75, Front End=0.6)
    )
    Sprite Rotation Mode=::NewEnumerator2
    Sprite Size Max=<3 3>
    Sprite Size Min=<1.5 1.5>
    Sprite Size Mode=::NewEnumerator2
    UsePositionOffset=false
    Write Material Random=true
   ) {module=InitializeParticle001}
   &PartitionParticles(Particles Per Partition=3, Partition Mode:ENiagara_PartitionMode_v2=::NewEnumerator1)
   &SkeletalMeshLocation(
    Bone / Socket Index=1
    Bone Sampling Mode=::NewEnumerator4
    MaskSpawnGroup=false
    Sampled Position Offset=&MaskVector(
     Mask Source=1
     Mask Parameter=$Particles.PartitionParticles.Partition
     Mask=0
     Masked Vector=&FN_FrontEndVector3Modifier(Gameplay=<-65 0 -5>, Front End=<-60 -30 -3>)
     Passthrough Vector=<1 1 1>
    )
    Sampled Position Offset Coordinate Space:ENiagaraSkelMeshTransforms=::NewEnumerator2
    Skeletal Mesh=$System.Skeletal Mesh
   )
   &ShapeLocation(
    MaskSpawnGroup=false
    Offset=&FN_FrontEndVector3Modifier(Gameplay=<0 0 0>, Front End=<0 0 0>)
    Offset Mode=::NewEnumerator0
    Override Seed=false
    Sphere Radius=2
    UseOwnerScale=true
   )
   /SetVariables(Particles.RockSize:NiagaraFloat=&RandomRangeFloat(
     Minimum=&FN_FrontEndFloatModifier(Gameplay=0.3, Front End=0.15)
     Maximum=&FN_FrontEndFloatModifier(Gameplay=0.9, Front End=0.55)
    )) {isSetParameters=true}
   /SetVariables(Particles.SpriteSize=&Multiply_Vector2D_ByFloat(Float=$Particles.RockSize, Vector2D=$Particles.SpriteSize)) {isSetParameters=true}
   /SetVariables(Particles.Scale=&Multiply_VectorByFloat(Vector=$Particles.Scale, Float=$Particles.RockSize)) {isSetParameters=true}
   &AddVelocityFromPoint(
    Module.Use Falloff Distance=false
    Origin Offset=<0 0 -1>
    Velocity Strength=&FN_FrontEndFloatModifier(Gameplay=5, Front End=1)
   )
   &AddVelocityInCone(
    Cone Angle=170
    Cone Axis=&RotateVectorByQuaternion(Vector To Rotate=<0 0 1>, Quaternion=<0 0 0 1>)
    Cone Axis Coordinate Space=::Simulation
    Module.Use Velocity Falloff On Cone Axis=true
    Velocity Distribution Along Cone Axis=1
    Velocity Falloff Away From Cone Axis=0.333
    Velocity Strength=&UniformRangedFloat(Minimum=20, Maximum=50)
   ) {module=AddVelocityInCone001}
   &AccelerationForce(Acceleration=&UniformRangedVector(
     Minimum=<0 0 -10>
     Maximum=&MakeVector(
      X=0
      Y=0
      Z=&FloatFromCurve(
       FloatCurve=DataInterface() {propertyValues {Curve {keys [<at 0.0174 1.2>,<at 0.9841 0.2965>]}}}
       CurveIndex=$Emitter.NormalizedLoopAge
       Scale Curve=45
      )
     )
    )) {module=AccelerationForce001}
   &AddRotationalVelocity(Rotation Rate=&RandomVector(Vector Scale=1))
   &AddVelocity(Velocity=&RandomRangeVector(
     Minimum=&FN_FrontEndVector3Modifier(Gameplay=<0 -35 5>, Front End=<0 -50 5>)
     Maximum=&FN_FrontEndVector3Modifier(Gameplay=<0 -100 35>, Front End=<0 -100 20>)
    ))
   &PointForce(
    Force Origin=&SimulationPosition()
    Force Strength=&UniformRangedFloat(Minimum=10, Maximum=&Multiply_Float(A=20, B=1))
    Module.Use Falloff=false
    Module.UseLinearForce=false
   )
   /SetVariables(Particles.Velocity=&Multiply_VectorByFloat(
     Vector=$Particles.Velocity
     Float=&FloatFromCurve(FloatCurve=DataInterface() {propertyValues {Curve {keys [<at 0.4 2>,<at 1.4 1>]}}}, CurveIndex=$Particles.RockSize)
    )) {isSetParameters=true}
   &ApplyInitialForces()
   /Niagara/Modules/Update/SubUV/SubUVAnimation(Random SubUV Lookup Index:NiagaraFloat=$Particles.RandomFloat, SubUV Animation Mode=::NewEnumerator1) {module=SubUVAnimation001}
   &InheritVelocity(Module.Limit Speed=true, Velocity Scale=&UniformRangedVector(Minimum=<0 -5 0>, Maximum=<0 -5 0>))
  ]
  particleUpdateScript [
   &ParticleState() {module=UpdateAge}
   &Drag(Use Linear Drag=true, Use Rotational Drag=false)
   &GravityForce(Gravity=&VectorFromCurve(
     VectorCurve=DataInterface() {propertyValues {XCurve {keys [<at 0 1>,<at 1 0>]}, YCurve {keys [<at 0 1>,<at 1 0>]}, ZCurve {keys [<at 0 1>,<at 1 -980>]}}}
     Scale Curve=<0 0 1>
    ))
   &Collision(
    Advanced Aging Rate=1
    Age Colliding Particles=false
    Collision Enabled=&Boolean_OrOperation(
     Bool1=&IsPlatformSetActive(New Platform Set=DataInterface() {propertyValues {Platforms {qualityLevelMask=-4}}})
     Bool2=$User.bIsFrontEndPreview
    )
    Control Roll On Collision=true
    Enable Rest State=false
    Module.EnableMaxCollisionCount=false
    Module.Randomize Collision Normal Vector=true
    Particle Radius Scale=0.5
    Restitution=0.4
    ScaleRateOfRotation=false
   )
   &SolveForcesAndVelocity(Clamp Velocity=false, Limit Acceleration=false)
   &SpriteRotationRate(Rotation Rate=&Multiply_Float(A=$Particles.RandomFloat, B=4))
   &ScaleSpriteSize(
    Uniform Curve Scale=2
    Uniform Curve Sprite Scale=DataInterface() {propertyValues {Curve {keys [
        <at 0 1>
        {arriveTangent=-1, interpMode=RCIM_Cubic, leaveTangent=-1, time=0.8, value=1}
        {arriveTangent=-5.7307, interpMode=RCIM_Cubic, leaveTangent=-5.7307, tangentMode=RCTM_User, time=1, value=0}
       ]}}}
   )
   &ScaleMeshSize(Scale Factor=&VectorFromFloat(Value=&FloatFromCurve(FloatCurve=DataInterface() {propertyValues {Curve {keys [
          <at 0 1>
          {arriveTangent=-1, interpMode=RCIM_Cubic, leaveTangent=-1, time=0.8, value=1}
          {arriveTangent=-5.7307, interpMode=RCIM_Cubic, leaveTangent=-5.7307, tangentMode=RCTM_User, time=1, value=0}
         ]}}})))
   &SpriteRotationRate(Rotation Rate=&RandomRangeFloat(Minimum=-300, Maximum=300)) {module=SpriteRotationRate001}
   /SetVariables(Particles.Position=&MakeCustomVectorFromBool(
     Bool=$User.bIsFrontEndPreview
     True Vector=&AddVectorToPosition(
      Position=$Particles.Position
      Vector=&Multiply_VectorByFloat(
       Vector=&RotateVectorByQuaternion(Vector To Rotate=<1 0 0>, Quaternion=$StackContext.SkeletalMeshLocation.SampledRotation)
       Float=-3
      )
     )
     False Vector=Linked() {linkedVariable {name=Particles.Position, type=NiagaraPosition}}
    )) {enabled=false, isSetParameters=true}
  ]
  rendererClasses [&NiagaraSpriteRendererProperties]
  renderers [{rendererClass=&NiagaraSpriteRendererProperties}]
 }
 Default_Emote_TraversalDust {
  bEnabled=false
  emitterSpawnScript [/SetVariables(Emitter.FE Particle Spawn Rate=60, Emitter.Max Traversal Speed=300, Emitter.Particle Spawn Rate=60) {isSetParameters=true}]
  emitterUpdateScript [
   &EmitterState()
   &SpawnRate(
    SpawnRate=&FN_FrontEndFloatModifier(
     Gameplay=&MakeCustomFloatFromBool(
      Bool=&SetBoolByFloatComparison(A=&VectorLength(Vector=$User.OwnerLocalVelocity), B=150)
      True Float=45
      False Float=10
     )
     Front End=45
    )
    Use Spawn Probability=false
   ) {module=SpawnRate002}
  ]
  particleSpawnScript [
   &PartitionParticles()
   &InitializeParticle(
    Color Channel Mode:ENiagaraChannelCorrelation=::NewEnumerator0
    Color Maximum=<rgba 0.781 0.7271 0.5498 0.4>
    Color Minimum=&UniformRangedLinearColor(Minimum=<rgba 0.2656 0.2656 0.2656 0.2>, Maximum=<rgba 0.2656 0.2656 0.2656 0.4>)
    Color Mode=::NewEnumerator2
    Lifetime Max=0.5
    Lifetime Min=0.3
    Lifetime Mode=::NewEnumerator1
    Sprite Rotation Mode=::NewEnumerator2
    Sprite Size Mode=::NewEnumerator1
    Uniform Sprite Size Max=50
    Uniform Sprite Size Min=30
    UsePositionOffset=false
    Write Material Random=true
   ) {module=InitializeParticle001}
   &SkeletalMeshLocation(
    Bone Sampling Mode=::NewEnumerator4
    MaskSpawnGroup=false
    Sampled Position Offset=&FN_FrontEndVector3Modifier(Gameplay=<0 0 -35>, Front End=<0 0 -30>)
    Sampled Position Offset Coordinate Space:ENiagaraSkelMeshTransforms=::NewEnumerator2
    Skeletal Mesh=$System.Skeletal Mesh
   )
   &SocketLocation(
    SamplePosOffset=true
    Sampled Position Offset=<0 -10 -17>
    Sampled Position Offset Coordinate Space:ENiagaraSkelMeshTransforms=::NewEnumerator2
    Socket Index=$Particles.PartitionParticles.Partition
    Socket Reader=$System.NewsVan Socket Reader
    Socket Sampling Mode=::NewEnumerator3
    VelocitySpeedLimit=false
   ) {enabled=false}
   &AddVelocity(Velocity=<0 2 0.2>, Velocity Speed Scale=&RandomRangeFloat(Minimum=300, Maximum=500)) {enabled=false, module=AddVelocity001}
   &CameraOffset(Camera Offset Amount=15)
   &AddVelocity(Velocity=&FN_FrontEndVector3Modifier(Gameplay=<0 -150 0>, Front End=<0 -150 0>))
  ]
  particleUpdateScript [
   &ParticleState()
   &Drag(
    Drag=&FloatFromCurve(
     FloatCurve=DataInterface() {propertyValues {Curve {
        defaultValue=1
        keys [{interpMode=RCIM_Cubic, tangentMode=RCTM_User, time=0, value=0.75}, {interpMode=RCIM_Cubic, tangentMode=RCTM_User, time=0.65, value=1}]
       }}}
     Scale Curve=&Lerp_Float(
      A=20
      B=8
      Alpha=&RemapRange(
       Input Value To Remap=&VectorLength(Vector=$User.OwnerLocalVelocity)
       Input Min=0
       Input Max=$Emitter.Max Traversal Speed
       Output Min=0
       Output Max=1
      )
     )
    )
    Use Linear Drag=true
    Use Rotational Drag=true
   )
   &ScaleSpriteSize(Uniform Curve Sprite Scale=DataInterface() {propertyValues {Curve {keys [
        {arriveTangent=0.8753, interpMode=RCIM_Cubic, leaveTangent=0.8753, tangentMode=RCTM_User, time=0, value=1}
        {arriveTangent=0.0238, interpMode=RCIM_Cubic, leaveTangent=0.0238, tangentMode=RCTM_User, time=1, value=1.3}
       ]}}})
   &DynamicMaterialParameters(
    Index 0 Param 1=&Add_Float(
     A=&RandomRangeFloat(Minimum=0, Maximum=1)
     B=&FloatFromCurve(
      FloatCurve=DataInterface() {propertyValues {Curve {keys [
          <at 0 0 RCIM_Cubic>
          {arriveTangent=1.5947, interpMode=RCIM_Cubic, leaveTangent=1.5947, tangentMode=RCTM_User, time=0.2557, value=0.5257}
          <at 1 1 RCIM_Cubic>
         ]}}}
      CurveIndex=$Particles.Age
      Scale Curve=0.25
     )
    )
    Module.Index0Float0Write=true
    Module.Index0Float1Write=true
    Module.Index0Float2Write=true
    Module.Index0Float3Write=true
   ) {module=DynamicMaterialParameters001}
   &ScaleColor(
    Scale Alpha=&FloatFromCurve(
     FloatCurve=DataInterface() {propertyValues {Curve {keys [
         {interpMode=RCIM_Cubic, tangentMode=RCTM_User, time=0, value=0}
         {arriveTangentWeight=0.0754, interpMode=RCIM_Cubic, leaveTangentWeight=0.2297, tangentMode=RCTM_User, time=0.1684, value=1}
         {arriveTangent=-0.5014, interpMode=RCIM_Cubic, leaveTangent=-0.5014, tangentMode=RCTM_User, time=0.4045, value=1}
         {arriveTangent=-2.9749, interpMode=RCIM_Cubic, leaveTangent=-2.9749, tangentMode=RCTM_User, time=0.9, value=0}
        ]}}}
     Scale Curve=&FN_FrontEndFloatModifier(Gameplay=0.4, Front End=0.4)
    )
    Scale RGB=&MakeVectorFromLinearColorRGB(Color=&FNTimeOfDayColor(
      Night Color=&RandomRangeLinearColor(Minimum=<rgba 0.4138 0.5201 0.6661 1>, Maximum=<rgba 0.781 0.781 0.781 1>)
      Morning Color=&RandomRangeLinearColor(Minimum=<rgba 0.6882 0.6227 0.9531 1>, Maximum=<rgba 0.6063 0.7457 0.9346 1>)
      Day Color=&RandomRangeLinearColor(Minimum=<rgba 0.478 0.4265 0.4121 0.8>, Maximum=<rgba 0.781 0.718 0.6943 0.9>)
      Evening Color=&RandomRangeLinearColor(Minimum=<rgba 0.478 0.478 0.478 1>, Maximum=<rgba 0.781 0.781 0.781 1>)
     ))
    ScaleA=true
    ScaleRGB=true
   )
   &BlendSpace(Rotation Contribution=0.5)
   &SolveForcesAndVelocity(Clamp Velocity=false, Limit Acceleration=false)
  ]
  rendererClasses [&NiagaraSpriteRendererProperties]
  renderers [
   {rendererClass=&NiagaraSpriteRendererProperties}
   {rendererClass=&NiagaraSpriteRendererProperties, rendererIndex=1}
  ]
 }
 NE_Smoke001 {
  bEnabled=false
  emitterUpdateScript [
   &EmitterState(Inactive Response=::NewEnumerator1, Life Cycle Mode=::NewEnumerator1, Loop Duration=1, UseLoopDelay=false)
   &SpawnRate(
    SpawnRate=&FN_FrontEndFloatModifier(
     Gameplay=&MakeCustomFloatFromBool(
      Bool=&SetBoolByFloatComparison(A=&VectorLength(Vector=$User.OwnerLocalVelocity), B=150)
      True Float=45
      False Float=10
     )
     Front End=45
    )
    Use Spawn Probability=false
   ) {module=SpawnRate002}
  ]
  particleSpawnScript [
   &InitializeParticle(
    Color=&UniformRangedLinearColor(Minimum=<rgba 0.2656 0.2656 0.2656 0.2>, Maximum=<rgba 0.2656 0.2656 0.2656 0.4>)
    Lifetime=&UniformRangedFloat(Minimum=0.6, Maximum=1.2)
    Sprite Rotation Angle=&UniformRangedFloat(Minimum=0, Maximum=360)
    Sprite Rotation Mode=::NewEnumerator0
    Sprite Size Mode=::NewEnumerator1
    Uniform Sprite Size Max=35
    Uniform Sprite Size Min=20
    UsePositionOffset=false
    Write Material Random=true
   )
   &SkeletalMeshLocation(
    Bone Sampling Mode=::NewEnumerator4
    MaskSpawnGroup=false
    Sampled Position Offset=&FN_FrontEndVector3Modifier(Gameplay=<0 0 -35>, Front End=<0 0 -30>)
    Sampled Position Offset Coordinate Space:ENiagaraSkelMeshTransforms=::NewEnumerator2
    Skeletal Mesh=$System.Skeletal Mesh
   )
   &AccelerationForce(Acceleration=&FN_FrontEndVector3Modifier(Gameplay=&RandomRangeVector(Minimum=<0 0 25>, Maximum=<0 0 50>), Front End=<0 0 10>))
   &ApplyInitialForces()
   &AddVelocity(Velocity=&FN_FrontEndVector3Modifier(Gameplay=<0 -150 0>, Front End=<0 -150 0>))
  ]
  particleUpdateScript [
   &ParticleState()
   &InheritSourceMovement(
    Applied Position Scale=&MakeCustomVectorFromBool(
     Bool=&SetBoolByFloatComparison(A=&VectorLength(Vector=$Engine.Owner.Velocity), B=210)
     True Vector=<0.3 0.5 1>
     False Vector=<-0.3 0.3 0.3>
    )
    ApplyForce=false
    ApplyPosition=true
    ApplyPositionFalloff=false
   )
   &Drag(
    Drag=&FN_FrontEndFloatModifier(
     Gameplay=&FloatFromCurve(FloatCurve=DataInterface() {propertyValues {Curve {
         defaultValue=1
         keys [
          {interpMode=RCIM_Cubic, tangentMode=RCTM_User, time=0, value=0}
          {arriveTangent=1, interpMode=RCIM_Cubic, leaveTangent=1, time=0.4, value=0.65}
          {interpMode=RCIM_Cubic, tangentMode=RCTM_User, time=1, value=1}
         ]
        }}})
     Front End=0.5
    )
    Use Linear Drag=true
    Use Rotational Drag=false
   )
   &SolveForcesAndVelocity(Clamp Velocity=false, Limit Acceleration=false)
   &ScaleSpriteSize(Scale Factor=&Vector2DFromFloat(Value=&FloatFromCurve(FloatCurve=DataInterface() {propertyValues {Curve {keys [
          {arriveTangent=1.491, interpMode=RCIM_Cubic, leaveTangent=1.491, tangentMode=RCTM_User, time=-0.0046, value=1.5186}
          <at 1.0011 3.4809 RCIM_Cubic>
         ]}}})))
   /SetVariables(Particles.SubImageIndex=&FloatFromCurve(FloatCurve=DataInterface() {propertyValues {Curve {keys [
         {arriveTangent=29.8623, interpMode=RCIM_Cubic, leaveTangent=29.8623, tangentMode=RCTM_User, time=-0.0046, value=5.0556}
         <at 1.0011 22.1689 RCIM_Cubic>
        ]}}})) {isSetParameters=true, module=SetVariables_#2}
   &DynamicMaterialParameters(
    Index 0 Param 1=&FloatFromCurve(FloatCurve=DataInterface() {propertyValues {Curve {keys [<at -0.0046 1.0113 RCIM_Cubic>,<at 1.0011 0.002>]}}})
    Index 0 Param 4=&FloatFromCurve(FloatCurve=DataInterface() {propertyValues {Curve {keys [<at 0 0 RCIM_Cubic>,<at 1 0.5>]}}})
    Module.Index0Float0Write=true
    Module.Index0Float1Write=true
    Module.Index0Float2Write=true
    Module.Index0Float3Write=true
   ) {module=DynamicMatParams}
   &CameraOffset(Camera Offset Amount=1)
   &ScaleColor(
    Scale Mode=::NewEnumerator1
    Scale RGBA=&MakeVector4_FromLinearColor(LinearColor=&FNTimeOfDayColor(
      Night Color=&RandomRangeLinearColor(Minimum=<rgba 0.1294 0.3996 0.8698 0.9>, Maximum=<rgba 0.4126 0.6462 1 1>)
      Morning Color=&RandomRangeLinearColor(Minimum=<rgba 0.2813 0.2326 0.478 0.1>, Maximum=<rgba 0.3377 0.5872 0.9253 0.3>)
      Day Color=&RandomRangeLinearColor(Minimum=<rgba 0.478 0.4265 0.4121 0.5>, Maximum=<rgba 0.781 0.718 0.6943 0.7>)
      Evening Color=&RandomRangeLinearColor(Minimum=<rgba 0.478 0.478 0.478 1>, Maximum=<rgba 0.781 0.781 0.781 1>)
     ))
    ScaleRGBA=true
   )
   &ScaleColor(
    Scale Alpha=&FloatFromCurve(
     FloatCurve=DataInterface() {propertyValues {Curve {keys [
         {arriveTangent=2.2146, interpMode=RCIM_Cubic, leaveTangent=2.2146, tangentMode=RCTM_User, time=0, value=0.25}
         {
          arriveTangent=0.0254
          arriveTangentWeight=0.4125
          interpMode=RCIM_Cubic
          leaveTangentWeight=0.1667
          tangentMode=RCTM_Break
          time=0.5
          value=1
         }
         {arriveTangent=-3.7464, tangentMode=RCTM_User, time=1, value=0}
        ]}}}
     Scale Curve=&MakeCustomFloatFromBool(Bool=$User.bIsFrontEnd, True Float=1, False Float=0.6)
    )
    ScaleA=true
    ScaleRGB=true
   ) {module=ScaleColor001}
  ]
  rendererClasses [&NiagaraSpriteRendererProperties]
  renderers [
   {rendererClass=&NiagaraSpriteRendererProperties}
   {rendererClass=&NiagaraSpriteRendererProperties, rendererIndex=1}
   {rendererClass=&NiagaraSpriteRendererProperties, rendererIndex=2}
  ]
 }
 SmokeGroundRings {
  bEnabled=false
  emitterUpdateScript [
   &EmitterState(
    Inactive Response=::NewEnumerator1
    Life Cycle Mode=::NewEnumerator1
    Loop Behavior=::NewEnumerator1
    Loop Duration Mode=::NewEnumerator1
    UseLoopDelay=false
   )
   &SpawnRate(
    SpawnRate=&FN_FrontEndFloatModifier(
     Gameplay=&Lerp_Float(
      A=4.5
      B=0
      Alpha=&MakeCustomFloatFromBool(
       Bool=&SetBoolByFloatComparison(
        A=&Add_Float(
         A=&Add_Float(
          A=&Abs_Float(Float:NiagaraFloat=&MakeFloatFromVector(Vector=$Engine.Owner.Velocity, Channel=::NewEnumerator0))
          B=&Abs_Float(Float:NiagaraFloat=&MakeFloatFromVector(Vector=$Engine.Owner.Velocity, Channel=::NewEnumerator2))
         )
         B=&Abs_Float(Float:NiagaraFloat=&MakeFloatFromVector(Vector=$Engine.Owner.Velocity, Channel=::NewEnumerator1))
        )
        B=40
       )
       True Float=&ClampFloat(Float=1)
       False Float=&ClampFloat(Float=0)
      )
     )
     Front End=0
    )
    Use Spawn Probability=false
   )
  ]
  particleSpawnScript [
   &InitializeParticle(
    Color=&FNTimeOfDayColor(
     Night Color=<rgba 0.5387 0.72 1 1.5>
     Morning Color=<rgba 1 0.7171 0.9907 1.5>
     Day Color=<rgba 1 0.7225 0.5208 1.5>
     Evening Color=<rgba 0.8049 1 0.8847 1.5>
    )
    Lifetime=&RandomRangeFloat(Minimum=0.9, Maximum=0.7)
    Mass Mode=::NewEnumerator1
    Sprite Rotation Angle=&UniformRangedFloat(Minimum=-180, Maximum=180)
    Sprite Rotation Mode=::NewEnumerator0
    Sprite Size=&Vector2DFromFloat(Value=300)
    Sprite Size Mode=::NewEnumerator4
    UsePositionOffset=false
    Write Material Random=true
   ) {module=InitializeParticle001}
   /SetVariables(Particles.SpriteFacing=<0 0 1>) {isSetParameters=true}
   &SystemLocation(MaskSpawnGroup=false, Offset=<0 0 -50>)
  ]
  particleUpdateScript [
   &ParticleState()
   &DynamicMaterialParameters(
    Index 0 Param 1=&FloatFromCurve(FloatCurve=DataInterface() {propertyValues {Curve {keys [<at 0 1.5364>,<at 1 2.501>]}}})
    Module.Param0WriteEnabled=true
    Module.Param1WriteEnabled=false
    Module.Param2WriteEnabled=false
    Module.Param3WriteEnabled=false
   )
   &ScaleSpriteSize(Scale Factor=&Vector2DFromFloat(Value=&FloatFromCurve(FloatCurve=DataInterface() {propertyValues {Curve {keys [
          {arriveTangent=0.6835, interpMode=RCIM_Cubic, leaveTangent=0.6835, tangentMode=RCTM_User, time=0, value=0.5706}
          {arriveTangentWeight=0.4822, interpMode=RCIM_Cubic, tangentMode=RCTM_User, time=1.0011, value=1}
         ]}}})))
   &CameraOffset(Camera Offset Amount=20)
   &ScaleColor(
    Scale Alpha=&FloatFromCurve(
     FloatCurve=DataInterface() {propertyValues {Curve {keys [
         {arriveTangent=0.3076, interpMode=RCIM_Cubic, leaveTangent=0.3076, tangentMode=RCTM_User, time=0, value=0.1879}
         {arriveTangent=-0.1857, interpMode=RCIM_Cubic, leaveTangent=-0.1857, time=0.4, value=0.4}
         {interpMode=RCIM_Cubic, leaveTangent=-4.9981, time=1.0011, value=0.002}
        ]}}}
     Scale Curve=&FNTimeOfDayScalar(Night Scale=0.14, Morning Scale=0.22, Day Scale=0.22, Evening Scale=0.21)
    )
    Scale RGB=&MakeVectorFromLinearColorRGB(Color=<rgba 1 1 1 1>)
    ScaleA=true
    ScaleRGB=true
   )
   &SolveForcesAndVelocity(Clamp Velocity=false, Limit Acceleration=false)
  ]
  rendererClasses [&NiagaraSpriteRendererProperties]
  renderers [{rendererClass=&NiagaraSpriteRendererProperties}]
 }
 Test {
  bEnabled=false
  emitterUpdateScript [
   &EmitterState(Inactive Response=::NewEnumerator1, Life Cycle Mode=::NewEnumerator1, Loop Duration=1, UseLoopDelay=false)
   &SpawnRate(SpawnRate=60, Use Spawn Probability=false) {module=SpawnRate002}
  ]
  particleSpawnScript [
   &PartitionParticles(Number Of Partitions=1)
   &InitializeParticle(
    Color=&UniformRangedLinearColor(Minimum=<rgba 0.478 0.3791 0.2199 0.07>, Maximum=<rgba 0.599 0.5576 0.4217 0.6>)
    Lifetime=&UniformRangedFloat(Minimum=0.6, Maximum=1.2)
    Sprite Rotation Angle=&UniformRangedFloat(Minimum=0, Maximum=360)
    Sprite Rotation Mode=::NewEnumerator0
    Sprite Size Mode=::NewEnumerator1
    Uniform Sprite Size Max=35
    Uniform Sprite Size Min=20
    UsePositionOffset=false
    Write Material Random=true
   )
   &SkeletalMeshLocation(
    Bone / Socket Index=$Particles.PartitionParticles.Partition
    Bone Sampling Mode=::NewEnumerator4
    MaskSpawnGroup=false
    Sampled Position Offset=&FN_FrontEndVector3Modifier(Gameplay=<0 0 0>, Front End=<0 0 0>)
    Sampled Position Offset Coordinate Space:ENiagaraSkelMeshTransforms=::NewEnumerator2
    Skeletal Mesh=$System.Skeletal Mesh
   )
   &AccelerationForce(Acceleration=&FN_FrontEndVector3Modifier(Gameplay=&RandomRangeVector(Minimum=<0 0 25>, Maximum=<0 0 50>), Front End=<0 0 10>))
   &ApplyInitialForces()
   &AddVelocity(Velocity=&FN_FrontEndVector3Modifier(Gameplay=<0 -150 0>, Front End=<0 -150 0>))
  ]
  particleUpdateScript [
   &ParticleState()
   &InheritSourceMovement(
    Applied Position Scale=&MakeCustomVectorFromBool(
     Bool=&SetBoolByFloatComparison(A=&VectorLength(Vector=$Engine.Owner.Velocity), B=210)
     True Vector=<0.3 0.5 1>
     False Vector=<-0.3 0.3 0.3>
    )
    ApplyForce=false
    ApplyPosition=true
    ApplyPositionFalloff=false
   )
   &Drag(
    Drag=&FN_FrontEndFloatModifier(
     Gameplay=&FloatFromCurve(FloatCurve=DataInterface() {propertyValues {Curve {
         defaultValue=1
         keys [
          {interpMode=RCIM_Cubic, tangentMode=RCTM_User, time=0, value=0}
          {arriveTangent=1, interpMode=RCIM_Cubic, leaveTangent=1, time=0.4, value=0.65}
          {interpMode=RCIM_Cubic, tangentMode=RCTM_User, time=1, value=1}
         ]
        }}})
     Front End=0.5
    )
    Use Linear Drag=true
    Use Rotational Drag=false
   )
   &SolveForcesAndVelocity(Clamp Velocity=false, Limit Acceleration=false)
   &ScaleSpriteSize(Scale Factor=&Vector2DFromFloat(Value=&FloatFromCurve(FloatCurve=DataInterface() {propertyValues {Curve {keys [
          {arriveTangent=1.491, interpMode=RCIM_Cubic, leaveTangent=1.491, tangentMode=RCTM_User, time=-0.0046, value=1.5186}
          <at 1.0011 3.4809 RCIM_Cubic>
         ]}}})))
   /SetVariables(Particles.SubImageIndex=&FloatFromCurve(FloatCurve=DataInterface() {propertyValues {Curve {keys [
         {arriveTangent=29.8623, interpMode=RCIM_Cubic, leaveTangent=29.8623, tangentMode=RCTM_User, time=-0.0046, value=5.0556}
         <at 1.0011 22.1689 RCIM_Cubic>
        ]}}})) {isSetParameters=true, module=SetVariables_#2}
   &DynamicMaterialParameters(
    Index 0 Param 1=&FloatFromCurve(FloatCurve=DataInterface() {propertyValues {Curve {keys [<at -0.0046 1.0113 RCIM_Cubic>,<at 1.0011 0.002>]}}})
    Index 0 Param 4=&FloatFromCurve(FloatCurve=DataInterface() {propertyValues {Curve {keys [<at 0 0 RCIM_Cubic>,<at 1 0.5>]}}})
    Module.Index0Float0Write=true
    Module.Index0Float1Write=true
    Module.Index0Float2Write=true
    Module.Index0Float3Write=true
   ) {module=DynamicMatParams}
   &CameraOffset(Camera Offset Amount=1)
   &ScaleColor(
    Scale Mode=::NewEnumerator1
    Scale RGBA=&MakeVector4_FromLinearColor(LinearColor=&FNTimeOfDayColor(
      Night Color=&RandomRangeLinearColor(Minimum=<rgba 0.1294 0.3996 0.8698 0.9>, Maximum=<rgba 0.4126 0.6462 1 1>)
      Morning Color=&RandomRangeLinearColor(Minimum=<rgba 0.2813 0.2326 0.478 0.1>, Maximum=<rgba 0.3377 0.5872 0.9253 0.3>)
      Day Color=&RandomRangeLinearColor(Minimum=<rgba 0.478 0.4265 0.4121 0.5>, Maximum=<rgba 0.781 0.718 0.6943 0.7>)
      Evening Color=&RandomRangeLinearColor(Minimum=<rgba 0.478 0.478 0.478 1>, Maximum=<rgba 0.781 0.781 0.781 1>)
     ))
    ScaleRGBA=true
   )
   &ScaleColor(
    Scale Alpha=&FloatFromCurve(
     FloatCurve=DataInterface() {propertyValues {Curve {keys [
         {arriveTangent=2.2146, interpMode=RCIM_Cubic, leaveTangent=2.2146, tangentMode=RCTM_User, time=0, value=0.25}
         {
          arriveTangent=0.0254
          arriveTangentWeight=0.4125
          interpMode=RCIM_Cubic
          leaveTangentWeight=0.1667
          tangentMode=RCTM_Break
          time=0.5
          value=1
         }
         {arriveTangent=-3.7464, tangentMode=RCTM_User, time=1, value=0}
        ]}}}
     Scale Curve=&MakeCustomFloatFromBool(Bool=$User.bIsFrontEnd, True Float=1, False Float=0.6)
    )
    ScaleA=true
    ScaleRGB=true
   ) {module=ScaleColor001}
  ]
  rendererClasses [&NiagaraSpriteRendererProperties]
  renderers [{rendererClass=&NiagaraSpriteRendererProperties}]
 }
}
summary {
 systemName=NS_Emote_ClaimReflect_TraversalDust
 emitters [
  {bEnabled=false, emitterName=NE_Smoke001, rendererClasses [&NiagaraSpriteRendererProperties]}
  {bEnabled=false, emitterName=Debris003, rendererClasses [&NiagaraSpriteRendererProperties]}
  {bEnabled=false, emitterName=Test, rendererClasses [&NiagaraSpriteRendererProperties]}
  {bEnabled=false, emitterName=SmokeGroundRings, rendererClasses [&NiagaraSpriteRendererProperties]}
  {bEnabled=false, emitterName=Default_Emote_TraversalDust, rendererClasses [&NiagaraSpriteRendererProperties]}
 ]
 userVariables [
  User.bIsFrontEnd {type=NiagaraBool}
  User.bIsFrontEndPreview {type=NiagaraBool}
  User.bIsInWater {type=NiagaraBool}
  User.Bool_Scale Alpha_Scale Curve_True Float {type=NiagaraBool, defaultValue=true}
  User.EffectScale {type=NiagaraFloat, defaultValue=1, description=""}
  User.OwnerLocalVelocity {type=Vector3f, description="", defaultValue {struct=Vector3f, value=<0 0 0>}}
  User.ParaGlideLeanAlpha {type=NiagaraFloat, defaultValue=0, description=""}
  User.WaterBodyIndex {type=NiagaraFloat, defaultValue=0, description=""}
  User.WaterDepth {type=NiagaraFloat, defaultValue=0, description=""}
  User.WaterPlaneLocation {type=NiagaraPosition, description="", defaultValue {}}
  User.WaterPlaneLocation {type=Vector3f, description="", defaultValue {struct=Vector3f, value=<0 0 0>}}
  User.WaterPlaneNormal {type=Vector3f, description="", defaultValue {struct=Vector3f, value=<0 0 0>}}
  User.WaveAttenuationFactor {type=NiagaraFloat, defaultValue=0, description=""}
  User.WaveTime {type=NiagaraFloat, defaultValue=0, description=""}
 ]
}
