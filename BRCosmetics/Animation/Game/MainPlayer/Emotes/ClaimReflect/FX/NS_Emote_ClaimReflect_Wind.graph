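# Niagara system for the ClaimReflect emote: wind, dust and thruster effects attached to the player's skeletal mesh.
# Velocity-gated Hover smoke and Traversal_Dust sprites fade out in front end.
# Long_Contrail ribbons and a merged thruster emitter, combining plasma, glow, lightning and lights, follow mesh bones or sockets.
# A system-level smoothed moving average blends contrail length and opacity plus thruster size and glow, while time-of-day scalars adjust dust and light.
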NiagaraSystem /BRCosmetics/Animation/Game/MainPlayer/Emotes/ClaimReflect/FX/NS_Emote_ClaimReflect_Wind
dependencies {
 usedDataInterfaces [
  &NiagaraDataInterfaceCurve
  &NiagaraDataInterfaceColorCurve
  &NiagaraDataInterfaceArrayFloat
  &NiagaraDataInterfaceArrayFloat3
  &NiagaraDataInterfaceSkeletalMesh
  &NiagaraDataInterfaceSocketReader
 ]
 usedDynamicInputs [&Vector2DFromFloat]
 usedModules [
  &self:SystemSpawnScript.SystemScriptSource.SystemScriptGraph.NiagaraNodeAssignment_0.SetVariables_#1
  &SystemState
  &BlendSpaceHelper
  &self:SystemSpawnScript.SystemScriptSource.SystemScriptGraph.NiagaraNodeAssignment_1.SetVariables_#2
  &InterpolateOverTime
  &self:SystemSpawnScript.SystemScriptSource.SystemScriptGraph.NiagaraNodeAssignment_3.SetVariables_#3
  &self:SystemSpawnScript.SystemScriptSource.SystemScriptGraph.NiagaraNodeAssignment_6.SetVariables_#4
  &self:SystemSpawnScript.SystemScriptSource.SystemScriptGraph.NiagaraNodeAssignment_9.SetVariables_#5
  &self:SystemSpawnScript.SystemScriptSource.SystemScriptGraph.NiagaraNodeAssignment_10.SetVariables_#6
  &self:SystemSpawnScript.SystemScriptSource.SystemScriptGraph.NiagaraNodeAssignment_2.SetVariables_#7
 ]
}
emitters {
 Hover {
  emitterUpdateScript [
   &EmitterState()
   &SpawnRate(
    SpawnRate=&Lerp_Float(
     A=&RandomRangeFloat(Minimum=0, Maximum=5, Evaluation Type=::NewEnumerator1)
     B=0
     Alpha=&MakeCustomFloatFromBool(
      Bool=&SetBoolByFloatComparison(
       A=&Add_Float(
        A=&Add_Float(
         A=&Abs_Float(Float:NiagaraFloat=&MakeFloatFromVector(Vector=$Engine.Owner.Velocity, Channel=::NewEnumerator0))
         B=&Abs_Float(Float:NiagaraFloat=&MakeFloatFromVector(Vector=$Engine.Owner.Velocity, Channel=::NewEnumerator2))
        )
        B=&Abs_Float(Float:NiagaraFloat=&MakeFloatFromVector(Vector=$Engine.Owner.Velocity, Channel=::NewEnumerator1))
       )
       B=170
      )
      True Float=&ClampFloat(Float=1)
      False Float=&ClampFloat(Float=0)
     )
    )
    Use Spawn Probability=false
   ) {module=SpawnRate001}
  ]
  particleSpawnScript [
   /SetVariables(Particles.Random=&RandomRangeFloat(Minimum=0, Maximum=1)) {isSetParameters=true}
   &CameraOffset(Camera Offset Amount=5)
   &InitializeParticle(
    Color=$System.SmokeCol
    Lifetime=&FN_FrontEndFloatModifier(Gameplay=1.1, Front End=0)
    Position Offset=<0 0 -85>
    Sprite Rotation Mode=::NewEnumerator2
    Sprite Size Mode=::NewEnumerator3
    Uniform Sprite Size=&Lerp_Float(A=130, B=180, Alpha=$Particles.Random)
    UsePositionOffset=true
    Write Material Random=true
   ) {module=InitializeParticle001}
   &ApplyInitialForces()
   &InitialMeshOrientation(
    Mesh Axis To Orient=<1 0.05 0.05>
    Mesh Orientation Mode=::NewEnumerator2
    Orientation Coordinate Space=::World
    Rotation=<0 -0.25 0>
    Use Rotation Vector=true
   )
   &SkeletalMeshLocation(
    Bone / Socket Index=6
    Bone Sampling Mode=::NewEnumerator4
    MaskSpawnGroup=false
    Orientation Sampling=::NewEnumerator1
    Position Sampling=::NewEnumerator1
    Sampled Position Offset=<0 0 -90>
    Skeletal Mesh=$System.Skeletal Mesh
   )
   /SetVariables(Particles.SpriteAlignment=<1 0 21.4746>) {enabled=false, isSetParameters=true}
  ]
  particleUpdateScript [
   &ParticleState()
   &ScaleSpriteSize(Uniform Curve Sprite Scale=DataInterface() {propertyValues {Curve {keys [
        {arriveTangent=1.6815, interpMode=RCIM_Cubic, leaveTangent=1.6815, tangentMode=RCTM_User, time=0.2, value=0.5}
        {arriveTangent=0.1063, interpMode=RCIM_Cubic, leaveTangent=0.1063, tangentMode=RCTM_User, time=0.55, value=1}
       ]}}})
   &SolveForcesAndVelocity(Clamp Velocity=false, Limit Acceleration=false)
   &AlignSpriteToMeshOrientation(SetAlignment=true, SetFacing=true)
   &ScaleColor(
    Scale Alpha=&FloatFromCurve(
     FloatCurve=DataInterface() {propertyValues {Curve {keys [<at 0 0>, {tangentMode=RCTM_User, time=0.1, value=0.9059}, {tangentMode=RCTM_User, time=0.7011, value=0.0094}]}}}
     Scale Curve=&FN_FrontEndFloatModifier(Gameplay=1, Front End=0)
    )
    Scale RGB=&MakeVectorFromLinearColorRGB(Color=&ColorFromCurve(ColorCurve=DataInterface() {propertyValues {
        AlphaCurve {keys [<at 0 0>,<at 0.4487 1 RCIM_Cubic>]}
        BlueCurve {keys [
          <at 0.1811 0.6425 RCIM_Cubic>
          {arriveTangent=-0.3057, interpMode=RCIM_Cubic, leaveTangent=-0.3057, time=0.4806, value=1}
          <at 0.7837 0.4583>
         ]}
        GreenCurve {keys [
          <at 0.1811 0.6562 RCIM_Cubic>
          {arriveTangent=-0.3285, interpMode=RCIM_Cubic, leaveTangent=-0.3285, time=0.4806, value=1}
          <at 0.7837 0.4583>
         ]}
        RedCurve {keys [
          <at 0.1811 0.6388 RCIM_Cubic>
          {arriveTangent=-0.2996, interpMode=RCIM_Cubic, leaveTangent=-0.2996, time=0.4806, value=1}
          <at 0.7837 0.4583>
         ]}
       }}))
    ScaleA=true
    ScaleRGB=true
   )
   &DynamicMaterialParameters(
    Index 0 Param 1=&FloatFromCurve(FloatCurve=DataInterface() {propertyValues {Curve {keys [<at 0 1.5364>,<at 1 2.501>]}}})
    Module.Param0WriteEnabled=true
    Module.Param1WriteEnabled=false
    Module.Param2WriteEnabled=false
    Module.Param3WriteEnabled=false
   ) {module=DynamicMaterialParameters002}
  ]
  rendererClasses [&NiagaraSpriteRendererProperties]
  renderers [{rendererClass=&NiagaraSpriteRendererProperties}]
 }
 Long_Contrail {
  emitterUpdateScript [
   &EmitterState(Life Cycle Mode=::NewEnumerator1, Loop Behavior=::NewEnumerator1, Loop Delay=0.07, Loop Duration Mode=::NewEnumerator1, UseLoopDelay=true)
   &FN_SpawnAnimTrail(Number of Anim Trails=2)
  ]
  particleSpawnScript [
   &FN_SetSpawnGroup(Number of Spawn Groups=2)
   &FN_InitializeAnimTrailBoneIndecies()
   &FN_InitializeAnimTrail(
    AnimTrailWidth=0.5
    Lifetime=&Lerp_Float(A=0.3, B=0.9, Alpha=$System.InterpolateOverTime.InterpolateOverTime_Float.Moving Average)
    SkeletalMesh=$System.Skeletal Mesh
   )
   &InitializeRibbon(
    Color=$System.Contrails
    Lifetime=&SelectFloatFromArray(
     Direct Array Index=&SpawnGroup()
     Float Selection Array=DataInterface() {propertyValues {FloatData [0.45,0.45]}}
    )
    Ribbon Facing Mode:ENiagara_UnsetDirectSet=::NewEnumerator1
    Ribbon Width=&SelectFloatFromArray(Direct Array Index=&SpawnGroup(), Float Selection Array=DataInterface() {propertyValues {FloatData [13,13]}})
    UsePositionOffset=false
    Write Material Random=true
   )
   &SkeletalMeshLocation(
    Bone / Socket Index=$Particles.SpawnGroup
    MaskSpawnGroup=false
    Mesh Sampling Type=::NewEnumerator5
    Orientation Sampling=::NewEnumerator1
    Sampled Position Offset=&MaskVector(Mask Source=1, Mask Parameter=&SpawnGroup(), Mask=0, Masked Vector=<0 0 0>, Passthrough Vector=<0 0 0>)
    Sampled Scale Coordinate Space:ENiagaraCoordinateSpace=::Simulation
    Skeletal Mesh=$System.Skeletal Mesh
    Socket Sampling Mode=::NewEnumerator5
   )
   &AddVelocity(Velocity=&MakeVector(
     X=0
     Y=&Lerp_Float(A=-100, B=-1400, Alpha=$System.InterpolateOverTime.InterpolateOverTime_Float.Moving Average)
     Z=0
    )) {module=AddVelocity001}
  ]
  particleUpdateScript [
   &ParticleState()
   &InheritSourceMovement(
    Applied Position Scale=&Multiply_VectorByFloat(
     Vector=<1 1 1>
     Float=&FloatFromCurve(FloatCurve=DataInterface() {propertyValues {Curve {keys [<at 0 1 RCIM_Cubic>, {interpMode=RCIM_Cubic, leaveTangent=-3.4326, time=0.7, value=0}]}}})
    )
    ApplyForce=false
    ApplyPosition=true
    ApplyPositionFalloff=false
   ) {module=InheritSourceMovement001}
   &SolveForcesAndVelocity(Clamp Velocity=false, Limit Acceleration=false)
   &ScaleColor(
    Scale Alpha=&Lerp_Float(A=0.5, B=1, Alpha=$System.InterpolateOverTime.InterpolateOverTime_Float.Moving Average)
    ScaleA=true
    ScaleRGB=true
   ) {module=ScaleColor001}
   &Color(
    Color=&ColorFromCurve(ColorCurve=DataInterface() {propertyValues {
       AlphaCurve {keys [<at 0 1 RCIM_Cubic>,<at 0.7463 0 RCIM_Cubic>]}
       BlueCurve {keys [<at 0 5 RCIM_Cubic>,<at 0.374 0.9238>,<at 0.9797 0.4281>]}
       GreenCurve {keys [<at 0 1.835 RCIM_Cubic>,<at 0.374 0.2715>,<at 0.9797 0.1258>]}
       RedCurve {keys [<at 0 0.6394 RCIM_Cubic>,<at 0.374 0.0494>,<at 0.9797 0.0229>]}
      }})
    Scale Alpha=&Multiply_Float(
     A=&FloatFromCurve(
      FloatCurve=DataInterface() {propertyValues {Curve {keys [
          <at 0 0 RCIM_Cubic>
          {interpMode=RCIM_Cubic, tangentMode=RCTM_User, time=0.2, value=1}
          {interpMode=RCIM_Cubic, tangentMode=RCTM_User, time=1, value=0}
         ]}}}
      Scale Curve=&FNTimeOfDayScalar(Night Scale=1, Morning Scale=1, Evening Scale=1)
     )
     B=1
    )
   ) {module=Color001}
   &FN_UpdateAnimTrail(SkeletalMesh=$System.Skeletal Mesh)
  ]
  rendererClasses [&NiagaraRibbonRendererProperties]
  renderers [{rendererClass=&NiagaraRibbonRendererProperties}]
 }
 Thruster_merged {
  emitterUpdateScript [
   &EmitterState(
    Inactive Response=::NewEnumerator1
    Life Cycle Mode=::NewEnumerator1
    Loop Behavior=::NewEnumerator1
    Loop Duration Mode=::NewEnumerator1
    UseLoopDelay=false
   )
   &SpawnBurst_Instantaneous(
    Spawn Group=6
    Spawn Probability=0.8
    Spawn Time=&RandomRangeFloat(Minimum=0, Maximum=1.2)
    Use Spawn Probability=true
    UseLoopCountLimit=false
   ) {module=SpawnBurst_Instantaneous005}
   &SpawnBurst_Instantaneous(
    Spawn Group=5
    Spawn Probability=0.5
    Spawn Time=&RandomRangeFloat(Minimum=0, Maximum=1.15)
    Use Spawn Probability=false
    UseLoopCountLimit=false
   ) {module=SpawnBurst_Instantaneous004}
   &SpawnBurst_Instantaneous(
    Spawn Group=4
    Spawn Time=&RandomRangeFloat(Minimum=0, Maximum=1.35)
    Use Spawn Probability=false
    UseLoopCountLimit=false
   ) {module=SpawnBurst_Instantaneous002}
   &SpawnBurst_Instantaneous(Spawn Count=4, Spawn Group=3, Use Spawn Probability=false, UseLoopCountLimit=false) {module=SpawnBurst_Instantaneous003}
   &SpawnBurst_Instantaneous(Spawn Group=1, Use Spawn Probability=false, UseLoopCountLimit=false) {module=SpawnBurst_Instantaneous001}
   &SpawnBurst_Instantaneous(Spawn Count=4, Use Spawn Probability=false, UseLoopCountLimit=false)
   /Niagara/Noise/Waveform(
    EnableExponent1=false
    Global Amplitude Scale=0
    GlobalClampMax=false
    GlobalClampMin=false
    UseInput1=false
    UsePhase1=false
    "Waveform [1]":ENiagara_Waveforms=::NewEnumerator12
    "[1] Frequency"=7
   )
   /SetVariables(Emitter.Flicker:NiagaraFloat=&Add_Float(A=&Multiply_Float(A=$Emitter.Waveform.Waveform, B=0.5), B=0.5)) {isSetParameters=true}
   /SetVariables(Emitter.LightVisibility:NiagaraBool=/Niagara/DynamicInputs/Helpers/MaskBool(
     Mask Source:NiagaraInt32=1
     Mask Parameter:NiagaraInt32=&MakeCustomIntFromBool(Bool=$User.bIsFrontEndPreview, True INT=1, False INT=0)
     Mask:NiagaraInt32=1
     Masked Bool:NiagaraBool=false
     Passthrough Bool:NiagaraBool=&SetBoolByFloatComparison(A=$User.LightVisibility, B=0.1)
    )) {isSetParameters=true}
   /SetVariables(Emitter.ColorCoreScalar=10) {isSetParameters=true}
  ]
  particleSpawnScript [
   /SetVariables(Particles.VisibilityTag=&SpawnGroup()) {isSetParameters=true}
   &FN_SetSpawnGroup(Number of Spawn Groups=5, Set Ribbon ID and Ribbon Link Order=true)
   &PartitionParticles(Number Of Partitions=5, Particles Per Partition=1, Partition Mode:ENiagara_PartitionMode_v2=::NewEnumerator1)
   &InitializeParticle(
    Color Mode=::NewEnumerator0
    Lifetime=&MergeHelper_FloatSwitch(
     0=1
     1=1
     2=1
     3=&Multiply_Float(
      A=&RandomRangeFloat(Minimum=0.4, Maximum=0.7)
      B=&FloatFromCurve(
       FloatCurve=DataInterface() {propertyValues {Curve {keys [<at 0 0.5>,<at 1.7 0.5>,<at 1.8 1>,<at 2 1>]}}}
       CurveIndex=$System.Age
       Scale Curve=2
      )
     )
     4=&Multiply_Float(
      A=&RandomRangeFloat(Minimum=0.4, Maximum=0.7)
      B=&FloatFromCurve(
       FloatCurve=DataInterface() {propertyValues {Curve {keys [<at 0 0.5>,<at 1.7 0.5>,<at 1.8 1>,<at 2 1>]}}}
       CurveIndex=$System.Age
       Scale Curve=2
      )
     )
     5=&Multiply_Float(
      A=&RandomRangeFloat(Minimum=0.4, Maximum=0.7)
      B=&FloatFromCurve(
       FloatCurve=DataInterface() {propertyValues {Curve {keys [<at 0 0.5>,<at 1.7 0.5>,<at 1.8 1>,<at 2 1>]}}}
       CurveIndex=$System.Age
       Scale Curve=2
      )
     )
     Index=&SpawnGroup()
     Num=::NewEnumerator4
    )
    Mesh Scale=&Multiply_VectorByFloat(Vector=<3.5 0.5 1.5>, Float=0.07)
    Mesh Scale Mode=::NewEnumerator4
    Sprite Rotation Mode=::NewEnumerator0
    Sprite Size=<90 175>
    Sprite Size Mode=::NewEnumerator4
    UsePositionOffset=false
    Write Material Random=true
   ) {module=InitializeParticle001}
   &InitialMeshOrientation(Mesh Axis To Orient=<0 0 1>, Mesh Orientation Mode=::NewEnumerator2, Orientation Vector=<0 1 0>, Use Rotation Vector=false)
   &CameraOffset(Camera Offset Amount=&MergeHelper_FloatSwitch(0=-5, 1=5, 2=0, 3=?, Index=0, Num=::NewEnumerator1))
   /SetVariables(Particles.GlobalRandom:NiagaraFloat=&RandomRangeFloat(Minimum=0.1, Maximum=0.23)) {isSetParameters=true}
   /SetVariables(
    "Particles.Lightning  Color":LinearColor=<rgba 0 0.5494 3 1>
    Particles.Lightning 2 position:NiagaraPosition=&AddVectorToPosition(Position=<-17.4229 25 -43>, Vector=<0 0 0>)
    Particles.Lightning 3 position:NiagaraPosition=&AddVectorToPosition(Position=<0 -35 -43>, Vector=<0 0 0>)
    Particles.Lightning MeshOrientation:Quat4f=<0 1 0 0>
    Particles.Lightning Position:NiagaraPosition=$StackContext.Position
    Particles.Lightning Scale:Vector3f=&VectorFromFloat(Value=&MakeFloatFromVector(
      Vector=&MakeVector(
       X=&RandomRangeFloat(Minimum=0.15, Maximum=0.2)
       Y=0.25
       Z=&Multiply_Float(A=0.3, B=$Particles.GlobalRandom)
      )
      Channel=::NewEnumerator0
     ))
   ) {isSetParameters=true}
  ]
  particleUpdateScript [
   &ParticleState(
    Kill Particles When Lifetime Has Elapsed=false
    Let Infinitely Lived Particles Die When Emitter Deactivates=true
    Loop Particles Lifetime=true
   )
   &SkeletalMeshLocation(
    Bone / Socket Index=$Particles.PartitionParticles.Partition
    MaskSpawnGroup=false
    Mesh Sampling Type=::NewEnumerator5
    Sampled Position Offset=&SelectVectorFromArray(
     Direct Array Index=&SpawnGroup()
     Vector Selection Array=DataInterface() {propertyValues {FloatData [<0 0 0>, <0 0 0>, <0 1.5 0>, <0 1.5 0>]}}
    )
    Sampled Scale Coordinate Space:ENiagaraCoordinateSpace=::Simulation
    Skeletal Mesh=$System.Skeletal Mesh
    Socket Sampling Mode=::NewEnumerator5
   )
   &InheritSourceMovement(
    Applied Position Scale=&Multiply_VectorByFloat(
     Vector=<1 1 1>
     Float=&FloatFromCurve(FloatCurve=DataInterface() {propertyValues {Curve {keys [<at 0 1 RCIM_Cubic>, {interpMode=RCIM_Cubic, leaveTangent=-3.4326, time=0.7, value=0}]}}})
    )
    ApplyForce=false
    ApplyPosition=true
    ApplyPositionFalloff=false
   ) {enabled=false, module=InheritSourceMovement001}
   &SolveForcesAndVelocity(Clamp Velocity=false, Limit Acceleration=false)
   &ScaleColor(Scale Alpha=&Multiply_Float(A=$Emitter.Flicker, B=3), ScaleA=true, ScaleRGB=true)
   &DynamicMaterialParameters(
    Index 0 Param 1=&RandomRangeFloat(Minimum=0, Maximum=1)
    Index 0 Param 4=&RandomRangeFloat(Minimum=-2, Maximum=-3)
    Module.Index0Param0Write=true
    Module.Index0Param1Write=false
    Module.Index0Param2Write=false
    Module.Index0Param3Write=true
   )
   /SetVariables(
    Particles.Plasma scale:Vector3f=&Multiply_VectorByFloat(
     Vector=&MaskVector(
      Mask Source=0
      Mask Parameter=?
      Mask=0
      Masked Vector=&Lerp_Vector(A=<0.25 0.07 0.1>, B=<0.25 0.05 0.17>, Alpha=$System.InterpolateOverTime.InterpolateOverTime_Float.Moving Average)
      Passthrough Vector=&MaskVector(
       Mask Source=0
       Mask Parameter=?
       Mask=1
       Masked Vector=&Lerp_Vector(A=<0.25 0.07 0.1>, B=<0.25 0.05 0.17>, Alpha=$System.InterpolateOverTime.InterpolateOverTime_Float.Moving Average)
       Passthrough Vector=&MaskVector(
        Mask Source=0
        Mask Parameter=?
        Mask=2
        Masked Vector=&Lerp_Vector(A=<0.04 0.04 0.04>, B=<0.04 0.04 0.05>, Alpha=$System.InterpolateOverTime.InterpolateOverTime_Float.Moving Average)
        Passthrough Vector=&Lerp_Vector(A=<0.04 0.04 0.04>, B=<0.04 0.04 0.05>, Alpha=$System.InterpolateOverTime.InterpolateOverTime_Float.Moving Average)
       )
      )
     )
     Float=&FloatFromCurve(
      FloatCurve=DataInterface() {propertyValues {Curve {
         defaultValue=1
         keys [
          {interpMode=RCIM_Cubic, tangentMode=RCTM_User, time=0, value=0}
          {interpMode=RCIM_Cubic, tangentMode=RCTM_User, time=0.1089, value=1}
          <at 0.2301 1 RCIM_Cubic>
         ]
        }}}
      CurveIndex=$Emitter.Age
     )
    )
    Particles.plasma color:LinearColor=<rgba 0 0.2783 2 0.1>
   ) {isSetParameters=true}
   /SetVariables(
    Particles.Glow color:LinearColor=&MakeLinearColorFromVectorAndFloat(
     "Vector (RGB)"=&Lerp_Vector(
      A=&MakeVectorFromLinearColorRGB(Color=<rgba 0 0.1831 1 1>)
      B=&MakeVectorFromLinearColorRGB(Color=<rgba 0 0.5494 3 1>)
      Alpha=$System.InterpolateOverTime.InterpolateOverTime_Float.Moving Average
     )
     "Float (Alpha)"=&Lerp_Float(A=0.8, B=1, Alpha=$System.InterpolateOverTime.InterpolateOverTime_Float.Moving Average)
    )
    Particles.Glow size:Vector2f=/Niagara/DynamicInputs/Helpers/MaskVector2d(
     Mask Source:NiagaraInt32=0
     Mask Parameter:NiagaraInt32=?
     Mask:NiagaraInt32=0
     Masked Vector2D:Vector2f=&Lerp_Vector2(A=<45 45>, B=<60 60>, Alpha=$System.InterpolateOverTime.InterpolateOverTime_Float.Moving Average)
     Passthrough Vector2D:Vector2f=&MaskVector2d(
      Mask Source:NiagaraInt32=0
      Mask Parameter:NiagaraInt32=?
      Mask:NiagaraInt32=1
      Masked Vector2D:Vector2f=&Lerp_Vector2(A=<45 45>, B=<60 60>, Alpha=$System.InterpolateOverTime.InterpolateOverTime_Float.Moving Average)
      Passthrough Vector2D:Vector2f=&MaskVector2d(
       Mask Source:NiagaraInt32=0
       Mask Parameter:NiagaraInt32=?
       Mask:NiagaraInt32=2
       Masked Vector2D:Vector2f=&Lerp_Vector2(A=<20 20>, B=<30 30>, Alpha=$System.InterpolateOverTime.InterpolateOverTime_Float.Moving Average)
       Passthrough Vector2D:Vector2f=&MaskVector2d(
        Mask Source:NiagaraInt32=0
        Mask Parameter:NiagaraInt32=?
        Mask:NiagaraInt32=3
        Masked Vector2D:Vector2f=&Lerp_Vector2(A=<20 20>, B=<30 30>, Alpha=$System.InterpolateOverTime.InterpolateOverTime_Float.Moving Average)
        Passthrough Vector2D:Vector2f=&MaskVector2d(
         Mask Source:NiagaraInt32=0
         Mask Parameter:NiagaraInt32=?
         Mask:NiagaraInt32=4
         Masked Vector2D:Vector2f=&Lerp_Vector2(A=<35 35>, B=<40 40>, Alpha=$System.InterpolateOverTime.InterpolateOverTime_Float.Moving Average)
         Passthrough Vector2D:Vector2f=&Lerp_Vector2(A=<35 35>, B=<40 40>, Alpha=$System.InterpolateOverTime.InterpolateOverTime_Float.Moving Average)
        )
       )
      )
     )
    )
   ) {isSetParameters=true}
   &FN_Light_Attributes(
    Color=<rgba 0.6394 1.835 5 1>
    Create Color Attribute=true
    Create Position Attribute=true
    Create Radius Attribute=true
    Enable Particle Light Enable=true
    Light Enabled:NiagaraBool=$Emitter.LightVisibility
    Position=<0 0 -60>
    Radius=50
    Scale Color=&Multiply_Float(
     A=$User.LightVisibility
     B=&Multiply_Float(
      A=/Niagara/DynamicInputs/Helpers/Waveform("Waveform [1]"=::NewEnumerator0, "[1] Frequency"=3, "[1] Amplitude Min/Max"=<0.7 1>, "[1] Scale"=2.6368)
      B=&FNTimeOfDayScalar(Night Scale=1, Morning Scale=0.8, Day Scale=0.9, Evening Scale=0.8)
     )
    )
   )
   /SetVariables(
    "Particles.Lightning  Color":LinearColor=&MakeLinearColorFromVectorAndFloat(
     "Vector (RGB)"=&MakeVectorFromLinearColorRGB(Color=<rgba 0 0.5494 3 1>)
     "Float (Alpha)"=&Waveform(Global Phase=$Particles.Age, "Waveform [1]"=::NewEnumerator0, "[1] Frequency"=10, "[1] Amplitude Min/Max"=<2 0.3>)
    )
    Particles.Lightning DynParam 1:Vector4f=&MakeVector4(
     X=&FloatFromCurve(
      FloatCurve=DataInterface() {propertyValues {Curve {
         defaultValue=0
         keys [
          {arriveTangent=9.0725, interpMode=RCIM_Cubic, leaveTangent=9.0725, tangentMode=RCTM_User, time=0, value=0}
          {arriveTangent=-2.5249, interpMode=RCIM_Cubic, leaveTangent=-2.5249, tangentMode=RCTM_User, time=0.1408, value=0.9847}
          {arriveTangent=-0.0405, interpMode=RCIM_Cubic, leaveTangent=-0.0405, tangentMode=RCTM_User, time=1, value=0.2}
         ]
        }}}
      Scale Curve=&RandomRangeFloat(Minimum=2, Maximum=1)
     )
     Y=&FloatFromCurve(
      FloatCurve=DataInterface() {propertyValues {Curve {
         defaultValue=0
         keys [<at 0 0.2 RCIM_Cubic>, {arriveTangent=2.4937, interpMode=RCIM_Cubic, leaveTangent=2.4937, tangentMode=RCTM_User, time=1, value=1}]
        }}}
      Scale Curve=&Multiply_Float(A=30, B=$Particles.GlobalRandom)
     )
     Z=&RandomRangeFloat(Minimum=0.25, Maximum=1)
     W=&FloatFromCurve(FloatCurve=DataInterface() {propertyValues {Curve {
         defaultValue=0
         keys [
          {arriveTangent=4.1785, interpMode=RCIM_Cubic, leaveTangent=4.1785, tangentMode=RCTM_User, time=0, value=0}
          {arriveTangent=0.0224, interpMode=RCIM_Cubic, leaveTangent=0.0224, tangentMode=RCTM_User, time=0.5, value=1}
         ]
        }}})
    )
    Particles.Lightning DynParam 2:Vector4f=&MakeVector4(
     X=&FloatFromCurve(FloatCurve=DataInterface() {propertyValues {Curve {
         defaultValue=1
         keys [
          {interpMode=RCIM_Cubic, tangentMode=RCTM_User, time=0, value=0}
          {arriveTangent=2.9607, interpMode=RCIM_Cubic, leaveTangent=2.9607, tangentMode=RCTM_User, time=0.7, value=1}
         ]
        }}})
     Y=&Multiply_Float(A=1, B=$Particles.GlobalRandom)
     Z=&Multiply_Float(A=0.5, B=$Particles.GlobalRandom)
     W=&RandomRangeFloat(Minimum=0.5, Maximum=1)
    )
    Particles.Lightning Position Update:NiagaraPosition=&Add_Vector(
     A=Linked() {linkedVariable {name=Particles.SkeletalMeshLocation.SampledPosition, type=NiagaraPosition}}
     B=&Multiply_VectorByFloat(
      Vector=&Add_Vector(A=$Engine.Owner.Velocity, B=<0 4.1144 0>)
      Float=&FloatFromCurve(FloatCurve=DataInterface() {propertyValues {Curve {defaultValue=0, keys [{tangentMode=RCTM_User, time=0, value=0}, {tangentMode=RCTM_User, time=1, value=1}]}}})
     )
    )
   ) {isSetParameters=true}
   /SetVariables(Particles.Thruster Scale:Vector3f=&Multiply_VectorByFloat(
     Vector=&MaskVector(
      Mask Source=0
      Mask Parameter=?
      Mask=0
      Masked Vector=&Lerp_Vector(A=<0.05 0.01 0.1>, B=<0.05 0.01 0.2>, Alpha=$System.InterpolateOverTime.InterpolateOverTime_Float.Moving Average)
      Passthrough Vector=&MaskVector(
       Mask Source=0
       Mask Parameter=?
       Mask=1
       Masked Vector=&Lerp_Vector(A=<0.05 0.01 0.1>, B=<0.05 0.01 0.2>, Alpha=$System.InterpolateOverTime.InterpolateOverTime_Float.Moving Average)
       Passthrough Vector=&MaskVector(
        Mask Source=0
        Mask Parameter=?
        Mask=2
        Masked Vector=&Lerp_Vector(A=<0.015 0.015 0.03>, B=<0.015 0.015 0.05>, Alpha=$System.InterpolateOverTime.InterpolateOverTime_Float.Moving Average)
        Passthrough Vector=&Lerp_Vector(A=<0.015 0.015 0.03>, B=<0.015 0.015 0.05>, Alpha=$System.InterpolateOverTime.InterpolateOverTime_Float.Moving Average)
       )
      )
     )
     Float=&FloatFromCurve(
      FloatCurve=DataInterface() {propertyValues {Curve {
         defaultValue=1
         keys [
          {interpMode=RCIM_Cubic, tangentMode=RCTM_User, time=0, value=0}
          {interpMode=RCIM_Cubic, tangentMode=RCTM_User, time=0.1089, value=1}
          <at 0.2301 1 RCIM_Cubic>
         ]
        }}}
      CurveIndex=$Emitter.Age
     )
    )) {isSetParameters=true}
  ]
  rendererClasses [
   &NiagaraMeshRendererProperties
   &NiagaraLightRendererProperties
   &NiagaraSpriteRendererProperties
  ]
  renderers [
   {rendererClass=&NiagaraMeshRendererProperties, rendererIndex=0}
   {rendererClass=&NiagaraMeshRendererProperties, rendererIndex=1}
   {rendererClass=&NiagaraLightRendererProperties, rendererIndex=2}
   {rendererClass=&NiagaraSpriteRendererProperties, rendererIndex=3}
   {rendererClass=&NiagaraMeshRendererProperties, rendererIndex=4}
   {rendererClass=&NiagaraMeshRendererProperties, rendererIndex=5}
   {rendererClass=&NiagaraMeshRendererProperties, rendererIndex=6}
  ]
 }
 Traversal_Dust {
  emitterSpawnScript [/SetVariables(Emitter.FE Particle Spawn Rate=60, Emitter.Max Traversal Speed=300, Emitter.Particle Spawn Rate=60) {isSetParameters=true}]
  emitterUpdateScript [
   &EmitterState()
   &SpawnRate(
    SpawnRate=&FN_FrontEndFloatModifier(
     Gameplay=&MakeCustomFloatFromBool(
      Bool=&SetBoolByFloatComparison(A=&VectorLength(Vector=$User.OwnerLocalVelocity), B=150)
      True Float=45
      False Float=10
     )
     Front End=45
    )
    Use Spawn Probability=false
   ) {module=SpawnRate002}
  ]
  particleSpawnScript [
   &PartitionParticles()
   &InitializeParticle(
    Color Channel Mode:ENiagaraChannelCorrelation=::NewEnumerator0
    Color Maximum=<rgba 0.5948 0.5948 0.5948 0.4>
    Color Minimum=&UniformRangedLinearColor(Minimum=<rgba 0.2656 0.2656 0.2656 0.2>, Maximum=<rgba 0.2656 0.2656 0.2656 0.4>)
    Color Mode=::NewEnumerator2
    Lifetime=&FN_FrontEndFloatModifier(Gameplay=&RandomRangeFloat(Minimum=0.3, Maximum=0.6), Front End=0)
    Position Offset=<0 0 -5>
    Sprite Rotation Mode=::NewEnumerator2
    Sprite Size Mode=::NewEnumerator1
    Uniform Sprite Size Max=65
    Uniform Sprite Size Min=45
    UsePositionOffset=true
    Write Material Random=true
   ) {module=InitializeParticle001}
   &SkeletalMeshLocation(
    Bone Sampling Mode=::NewEnumerator4
    MaskSpawnGroup=false
    Sampled Position Offset=&FN_FrontEndVector3Modifier(Gameplay=<0 -45 -35>, Front End=<0 0 -30>)
    Sampled Position Offset Coordinate Space:ENiagaraSkelMeshTransforms=::NewEnumerator2
    Skeletal Mesh=$System.Skeletal Mesh
   )
   &AddVelocity(Velocity=<0 2 0.2>, Velocity Speed Scale=&RandomRangeFloat(Minimum=300, Maximum=500)) {enabled=false, module=AddVelocity001}
   &CameraOffset(Camera Offset Amount=15)
   &AddVelocity(Velocity=&FN_FrontEndVector3Modifier(Gameplay=<0 -150 0>, Front End=<0 0 0>))
   &ShapeLocation(
    MaskSpawnGroup=false
    Override Seed=false
    Ring / Disc Mode:ENiagaraRingDiscMode=::NewEnumerator1
    Ring Radius=10
    Shape Primitive=::NewEnumerator4
    UseOwnerScale=true
   )
  ]
  particleUpdateScript [
   &ParticleState()
   &Drag(
    Drag=&FloatFromCurve(
     FloatCurve=DataInterface() {propertyValues {Curve {
        defaultValue=1
        keys [{interpMode=RCIM_Cubic, tangentMode=RCTM_User, time=0, value=0.75}, {interpMode=RCIM_Cubic, tangentMode=RCTM_User, time=0.65, value=1}]
       }}}
     Scale Curve=&Lerp_Float(
      A=2
      B=8
      Alpha=&RemapRange(
       Input Value To Remap=&VectorLength(Vector=$User.OwnerLocalVelocity)
       Input Min=0
       Input Max=$Emitter.Max Traversal Speed
       Output Min=0
       Output Max=1
      )
     )
    )
    Use Linear Drag=true
    Use Rotational Drag=true
   )
   &ScaleSpriteSize(Uniform Curve Sprite Scale=DataInterface() {propertyValues {Curve {keys [
        {arriveTangent=0.8753, interpMode=RCIM_Cubic, leaveTangent=0.8753, tangentMode=RCTM_User, time=0, value=1}
        {arriveTangent=0.0238, interpMode=RCIM_Cubic, leaveTangent=0.0238, tangentMode=RCTM_User, time=1, value=1.3}
       ]}}})
   &DynamicMaterialParameters(
    Index 0 Param 1=&Add_Float(
     A=&RandomRangeFloat(Minimum=0, Maximum=1)
     B=&FloatFromCurve(
      FloatCurve=DataInterface() {propertyValues {Curve {keys [
          <at 0 0 RCIM_Cubic>
          {arriveTangent=1.5947, interpMode=RCIM_Cubic, leaveTangent=1.5947, tangentMode=RCTM_User, time=0.2557, value=0.5257}
          <at 1 1 RCIM_Cubic>
         ]}}}
      CurveIndex=$Particles.Age
      Scale Curve=0.25
     )
    )
    Module.Index0Float0Write=true
    Module.Index0Float1Write=true
    Module.Index0Float2Write=true
    Module.Index0Float3Write=true
   ) {module=DynamicMaterialParameters001}
   &ScaleColor(
    Scale Alpha=&FloatFromCurve(
     FloatCurve=DataInterface() {propertyValues {Curve {keys [
         {interpMode=RCIM_Cubic, tangentMode=RCTM_User, time=0, value=0}
         {arriveTangentWeight=0.0754, interpMode=RCIM_Cubic, leaveTangentWeight=0.2297, tangentMode=RCTM_User, time=0.1684, value=1}
         {arriveTangent=-0.5014, interpMode=RCIM_Cubic, leaveTangent=-0.5014, tangentMode=RCTM_User, time=0.4045, value=1}
         {arriveTangent=-2.9749, interpMode=RCIM_Cubic, leaveTangent=-2.9749, tangentMode=RCTM_User, time=0.9, value=0}
        ]}}}
     Scale Curve=&FN_FrontEndFloatModifier(Gameplay=0.4, Front End=0)
    )
    Scale RGB=&MakeVectorFromLinearColorRGB(Color=&FNTimeOfDayColor(
      Night Color=&RandomRangeLinearColor(Minimum=<rgba 0.4138 0.5201 0.6661 1>, Maximum=<rgba 0.781 0.781 0.781 1>)
      Morning Color=&RandomRangeLinearColor(Minimum=<rgba 0.6882 0.6227 0.9531 1>, Maximum=<rgba 0.6063 0.7457 0.9346 1>)
      Day Color=&RandomRangeLinearColor(Minimum=<rgba 0.4243 0.4243 0.4243 1>, Maximum=<rgba 0.5948 0.5948 0.5948 1>)
      Evening Color=&RandomRangeLinearColor(Minimum=<rgba 0.478 0.478 0.478 1>, Maximum=<rgba 0.781 0.781 0.781 1>)
     ))
    ScaleA=true
    ScaleRGB=true
   )
   &BlendSpace(Rotation Contribution=0.5)
   &SolveForcesAndVelocity(Clamp Velocity=false, Limit Acceleration=false)
  ]
  rendererClasses [&NiagaraSpriteRendererProperties]
  renderers [
   {rendererClass=&NiagaraSpriteRendererProperties}
   {rendererClass=&NiagaraSpriteRendererProperties, rendererIndex=1}
  ]
 }
}
summary {
 systemName=NS_Emote_ClaimReflect_Wind
 emitters [
  {emitterName=Hover, rendererClasses [&NiagaraSpriteRendererProperties]}
  {emitterName=Traversal_Dust, rendererClasses [&NiagaraSpriteRendererProperties]}
  {emitterName=Long_Contrail, rendererClasses [&NiagaraRibbonRendererProperties]}
  {
   emitterName=Thruster_merged
   rendererClasses [
    &NiagaraMeshRendererProperties
    &NiagaraLightRendererProperties
    &NiagaraSpriteRendererProperties
   ]
  }
 ]
 userVariables [
  User.Anim_Curve_SpawnRate_Ribbon {type=NiagaraFloat, defaultValue=1, description=""}
  User.bIsFrontEnd {type=NiagaraBool}
  User.bIsFrontEndPreview {type=NiagaraBool}
  User.ForwardDot {type=NiagaraFloat, defaultValue=1, description=""}
  User.GlobalScale {type=NiagaraFloat, defaultValue=2, description=""}
  User.LightVisibility {type=NiagaraFloat, defaultValue=1, description=""}
  User.OwnerLocalVelocity {type=Vector3f, description="", defaultValue {struct=Vector3f, value=<0 0 0>}}
  User.PlayerSpeed {type=NiagaraFloat, defaultValue=0, description=""}
  User.RibbonIDOffset {type=NiagaraInt32}
 ]
}
